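# Niagara system for the idle head effect on the TallChore player character: a glowing green eye and wispy flames.
# Both emitters burst-spawn persistent particles that track the character through a shared skeletal mesh reader.
# Eye is one offset sprite; WhispFlames puts four particles on sockets by exec index, varying mirrored scales from an array across sprite and layered mesh renderers.
# A user bool exposes whether the head is hidden.
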NiagaraSystem /BRCosmetics/Characters/Player/Male/Large/Bodies/M_LRG_TallChore/FX/NS_Character_TallChore_Idle_Head
dependencies {
 usedDataInterfaces [&NiagaraDataInterfaceArrayFloat3, &NiagaraDataInterfaceSkeletalMesh]
 usedDynamicInputs [&Boolean_NotOperation]
 usedModules [
  &self:SystemSpawnScript.SystemScriptSource.SystemScriptGraph.NiagaraNodeAssignment_0.SetVariables_#1
  &SystemState
  &self:SystemSpawnScript.SystemScriptSource.SystemScriptGraph.NiagaraNodeAssignment_1.SetVariables_#2
 ]
}
emitters {
 Eye {
  emitterUpdateScript [
   &EmitterState(Life Cycle Mode=::NewEnumerator1, Loop Behavior=::NewEnumerator1, Loop Duration Mode=::NewEnumerator1, UseLoopDelay=false)
   &SpawnBurst_Instantaneous(Use Spawn Probability=false, UseLoopCountLimit=false)
  ]
  particleSpawnScript [&InitializeParticle(Color=<rgba 0.4033 1 0.4984 1>, Sprite Size=<5 4>, Sprite Size Mode=::NewEnumerator4, UsePositionOffset=false, Write Material Random=true)]
  particleUpdateScript [
   &ParticleState(Kill Particles When Lifetime Has Elapsed=false, Let Infinitely Lived Particles Die When Emitter Deactivates=true)
   &SkeletalMeshLocation(Bone Sampling Mode=::NewEnumerator4, MaskSpawnGroup=false, Sampled Position Offset=<1 22 0>, Skeletal Mesh=$System.Skeletal Mesh Reader)
  ]
  rendererClasses [&NiagaraSpriteRendererProperties]
  renderers [{rendererClass=&NiagaraSpriteRendererProperties}]
 }
 WhispFlames {
  emitterUpdateScript [
   &EmitterState(Life Cycle Mode=::NewEnumerator1, Loop Behavior=::NewEnumerator1, Loop Duration Mode=::NewEnumerator1, UseLoopDelay=false)
   &SpawnBurst_Instantaneous(Spawn Count=4, Use Spawn Probability=false, UseLoopCountLimit=false)
  ]
  particleSpawnScript [&InitializeParticle(
    Color=<rgba 1 1 1 0.3>
    Lifetime Mode=::NewEnumerator1
    Mesh Scale=&Multiply_VectorByFloat(
     Vector=&SelectVectorFromArray(
      Direct Array Index=&ReturnExecIndex()
      Vector Selection Array=DataInterface() {propertyValues {FloatData [<-0.9 0.9 0.5>, <0.9 0.9 0.5>, <0.7 -0.7 0.6>, <0.7 0.7 0.6>]}}
     )
     Float=1
    )
    Mesh Scale Mode=::NewEnumerator4
    Sprite Size Mode=::NewEnumerator3
    Uniform Sprite Size=2
    UsePositionOffset=false
    Write Material Random=true
   )]
  particleUpdateScript [
   &ParticleState(Kill Particles When Lifetime Has Elapsed=false)
   &CameraOffset(Camera Offset Amount=12)
   /SetVariables(Particles.Color2=<rgba 0.0323 0.1233 0.0495 0.3>) {isSetParameters=true}
   /SetVariables(Particles.Color3:LinearColor=<rgba 0.3864 0.6172 0.4342 0.2>) {isSetParameters=true}
   /SetVariables(Particles.GlowScale:Vector3f=&Multiply_Vector(A=$Particles.Scale, B=<1.2 1.2 0.96>)) {isSetParameters=true}
   &SkeletalMeshLocation(
    Bone / Socket Index=&ReturnExecIndex()
    MaskSpawnGroup=false
    Mesh Sampling Type=::NewEnumerator5
    Skeletal Mesh=$System.Skeletal Mesh Reader
    Socket Sampling Mode=::NewEnumerator5
   )
  ]
  rendererClasses [&NiagaraSpriteRendererProperties, &NiagaraMeshRendererProperties]
  renderers [
   {rendererClass=&NiagaraSpriteRendererProperties}
   {rendererClass=&NiagaraMeshRendererProperties, rendererIndex=1}
   {rendererClass=&NiagaraMeshRendererProperties, rendererIndex=2}
   {rendererClass=&NiagaraMeshRendererProperties, rendererIndex=3}
  ]
 }
}
summary {
 systemName=NS_Character_TallChore_Idle_Head
 emitters [
  {emitterName=WhispFlames, rendererClasses [&NiagaraSpriteRendererProperties, &NiagaraMeshRendererProperties]}
  {emitterName=Eye, rendererClasses [&NiagaraSpriteRendererProperties]}
 ]
 userVariables [User.bisHeadHidden {type=NiagaraBool}]
}
