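# Control Rig for the TallChore large male player body, procedurally animating its tentacles and remapping legs onto dog-style leg bones.
# Construction spawns toe controls, pole vectors and duplicate IK leg chains.
# Forward solve runs FortDeformRig and phantom finger constraints, then poses tentacles by aim, head angle and skydiving, adding harmonic waves.
# Without custom boots, stacked two-bone IK drives the dog legs, with offsets blended by pelvis height.
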ControlRigBlueprint /BRCosmetics/Characters/Player/Male/Large/Bodies/M_LRG_TallChore/Meshes/M_LRG_TallChore_CtrlRig
graphs [
 ContainedGraph {nodes [
   n1 = RigVMFunctionEntryNode()
   n2 = &RigUnit_GetTransform(Space=LocalSpace, Item=@n1.Item)
   n3 = &RigUnit_GetTransform(Space=LocalSpace, bInitial=true, Item=@n1.Item)
   > RigVMFunctionReturnNode(OutTransform=/Script/RigVM.RigVMFunction_MathTransformToEulerTransform(Value=&RigVMFunction_MathTransformMakeRelative(Global=@n2.Transform, Parent=@n3.Transform).Local) {is=struct RigVMUnitNode, template="Cast::Execute(in Value,out Result)"}.Result)
  ]}
 New Function_ContainedGraph {nodes [
   n1 = RigVMFunctionEntryNode()
   RigVMFunction_MathQuaternionToEuler = &RigVMFunction_MathQuaternionToEuler(
    RotationOrder=XYZ
    Value=&RigUnit_GetRelativeTransformForItem(
     Child=(Type=Bone,Name="hand_l")
     Parent=(Type=Null,Name="lowerarm_l")
     bChildInitial=False
     bParentInitial=False
     Child.Name=@n1.Child
     Parent.Name=@n1.Parent
    ).RelativeTransform.Rotation
   )
   Remap = &RigVMFunction_MathDoubleRemap(
    TargetMaximum=1
    TargetMinimum=0
    Value=@RigVMFunction_MathQuaternionToEuler.Result.Y
    SourceMaximum=@n1.Source Maximum
    SourceMinimum=@n1.Source Minimum
   )
   > RigVMFunctionReturnNode(Delta=@Remap.Result)
  ]}
 New Function_ContainedGraph {nodes [
   n1 = RigVMFunctionEntryNode()
   RigVMFunction_MathQuaternionToEuler = &RigVMFunction_MathQuaternionToEuler(
    RotationOrder=XYZ
    Value=&RigUnit_GetRelativeTransformForItem(
     Child=(Type=Bone,Name="hand_l")
     Parent=(Type=Null,Name="lowerarm_l")
     bChildInitial=False
     bParentInitial=False
     Child.Name=@n1.Child
     Parent.Name=@n1.Parent
    ).RelativeTransform.Rotation
   )
   Remap = &RigVMFunction_MathDoubleRemap(
    TargetMaximum=1
    TargetMinimum=0
    Value=@RigVMFunction_MathQuaternionToEuler.Result.X
    SourceMaximum=@n1.Source Maximum
    SourceMinimum=@n1.Source Minimum
   )
   > RigVMFunctionReturnNode(Delta=@Remap.Result)
  ]}
 RigVMFunctionLibrary {nodes [
   DeformRig_GetDeltaRefTransform_FNC = RigVMCollapseNode(ExecuteContext=(), Item=(Type=Bone))
   DeltaRotationX = RigVMCollapseNode(ExecuteContext=())
   DeltaRotationY = RigVMCollapseNode(ExecuteContext=())
  ]}
 RigVMFunction_Sequence_1_ContainedGraph {nodes [
   n1 = RigVMFunctionEntryNode()
   RigVMFunction_Sequence = &RigVMFunction_Sequence(ExecuteContext=@n1.ExecuteContext)
   n2 = &RigVMFunction_Sequence(ExecuteContext=@RigVMFunction_Sequence.B)
   n3 = &RigVMFunction_Sequence(ExecuteContext=@n2.B)
   RigVMFunctionReturnNode(A=@RigVMFunction_Sequence.A, B=@n2.A, C=@n3.A, D=@n3.B)
  ]}
 RigVMFunction_Sequence_1_ContainedGraph {nodes [
   n1 = RigVMFunctionEntryNode()
   n2 = &RigVMFunction_Sequence(ExecuteContext=@n1.ExecuteContext)
   n3 = &RigVMFunction_Sequence(ExecuteContext=@n2.B)
   n4 = &RigVMFunction_Sequence(ExecuteContext=@n3.B)
   n5 = &RigVMFunction_Sequence(ExecuteContext=@n4.B)
   RigVMFunctionReturnNode(B=@n3.A, C=@n4.A, D=@n5.A, E=@n5.B, A=@n2.A)
  ]}
 RigVMFunction_Sequence_2_ContainedGraph {nodes [
   n1 = RigVMFunctionEntryNode()
   RigVMFunction_Sequence = &RigVMFunction_Sequence(ExecuteContext=@n1.ExecuteContext)
   n2 = &RigVMFunction_Sequence(ExecuteContext=@RigVMFunction_Sequence.B)
   n3 = &RigVMFunction_Sequence(ExecuteContext=@n2.B)
   RigVMFunctionReturnNode(A=@RigVMFunction_Sequence.A, B=@n2.A, C=@n3.A, D=@n3.B)
  ]}
 RigVMFunction_Sequence_4_ContainedGraph {nodes [
   n1 = RigVMFunctionEntryNode()
   n2 = &RigVMFunction_Sequence(ExecuteContext=@n1.ExecuteContext)
   n3 = &RigVMFunction_Sequence(ExecuteContext=@n2.B)
   RigVMFunctionReturnNode(B=@n3.A, C=@n3.B, A=@n2.A)
  ]}
 RigVMFunction_Sequence_4_ContainedGraph {nodes [
   n1 = RigVMFunctionEntryNode()
   n2 = &RigVMFunction_Sequence(ExecuteContext=@n1.ExecuteContext)
   n3 = &RigVMFunction_Sequence(ExecuteContext=@n2.B)
   n4 = &RigVMFunction_Sequence(ExecuteContext=@n3.B)
   RigVMFunctionReturnNode(B=@n3.A, C=@n4.A, D=@n4.B, A=@n2.A)
  ]}
 RigVMModel {nodes [
   n1 = &RigUnit_PrepareForExecution()
   &RigUnit_HierarchyImportFromSkeleton(NameSpace=None, bIncludeCurves=False, bIncludeMeshSockets=False, bIncludeVirtualBones=True, ExecutePin=@n1.ExecutePin)
   > RigVMFunction_Sequence_3 = RigVMAggregateNode()
   For_Each = RigVMDispatchNode(ExecuteContext=@RigVMFunction_Sequence_3.A, Array=RigVMVariableNode() {variable=Toes}.Value) {template="DISPATCH_RigVMDispatch_ArrayIterator(in Array,out Element,out Index,out Count,out Ratio)"}
   To_String = RigVMDispatchNode(Value=@For_Each.Element.Name) {template="DISPATCH_RigDispatch_ToString(in Value,out Result)"}
   RigVMFunction_StringJoin = /Script/RigVM.RigVMFunction_StringJoin(Separator=_, Values=("ctrl",""), Values.1=@To_String.Result) {is=struct RigVMUnitNode}
   From_String = RigVMDispatchNode(String=@RigVMFunction_StringJoin.Result) {template="DISPATCH_RigDispatch_FromString(in String,out Result)"}
   > /Script/ControlRig.RigUnit_HierarchyAddBone(
    Parent=(Type=Bone,Name="root")
    Space=GlobalSpace
    Name=@From_String.Result
    Transform=&RigUnit_GetTransform(bInitial=False, Item=@For_Each.Element).Transform
   ) {is=struct RigVMUnitNode}
   &RigUnit_HierarchyAddBone(
    Name=PV_L
    Parent=(Type=Bone,Name="root")
    Space=GlobalSpace
    Transform=(Rotation=(X=0,Y=0,Z=0,W=1),Translation=(X=0,Y=0,Z=0),Scale3D=(X=1,Y=1,Z=1))
    ExecutePin=@For_Each.Completed
   )
   > &RigUnit_HierarchyAddBone(
    Name=PV_R
    Parent=(Type=Bone,Name="root")
    Space=GlobalSpace
    Transform=(Rotation=(X=0,Y=0,Z=0,W=1),Translation=(X=0,Y=0,Z=0),Scale3D=(X=1,Y=1,Z=1))
   )
   > &RigUnit_HierarchyAddBone(
    Name=PV_L1
    Parent=(Type=Bone,Name="root")
    Space=GlobalSpace
    Transform=(Rotation=(X=0,Y=0,Z=0,W=1),Translation=(X=0,Y=0,Z=0),Scale3D=(X=1,Y=1,Z=1))
   )
   > &RigUnit_HierarchyAddBone(
    Name=PV_R1
    Parent=(Type=Bone,Name="root")
    Space=GlobalSpace
    Transform=(Rotation=(X=0,Y=0,Z=0,W=1),Translation=(X=0,Y=0,Z=0),Scale3D=(X=1,Y=1,Z=1))
   )
   n2 = RigVMVariableNode() {variable=IK1_L}
   For_Each_1 = RigVMDispatchNode(
    Array=((Type=None,Name="None"),(Type=None,Name="None"),(Type=None,Name="None"))
    Array.0=RigVMDispatchNode(Index=0, Array=@n2.Value) {template="DISPATCH_RigVMDispatch_ArrayGetAtIndex(in Array,in Index,out Element)"}.Element
    Array.1=RigVMDispatchNode(Index=1, Array=@n2.Value) {template="DISPATCH_RigVMDispatch_ArrayGetAtIndex(in Array,in Index,out Element)"}.Element
    Array.2=RigVMDispatchNode(Index=3, Array=@n2.Value) {template="DISPATCH_RigVMDispatch_ArrayGetAtIndex(in Array,in Index,out Element)"}.Element
    ExecuteContext=@RigVMFunction_Sequence_3.B
   ) {template="DISPATCH_RigVMDispatch_ArrayIterator(in Array,out Element,out Index,out Count,out Ratio)"}
   To_String_1 = RigVMDispatchNode(Value=@For_Each_1.Element.Name) {template="DISPATCH_RigDispatch_ToString(in Value,out Result)"}
   From_String_1 = RigVMDispatchNode(String=&RigVMFunction_StringJoin(Separator=_, Values=("IK1",""), Values.1=@To_String_1.Result).Result) {template="DISPATCH_RigDispatch_FromString(in String,out Result)"}
   n3 = RigVMVariableNode() {variable=IK1_L_spawned}
   > n4 = &RigUnit_HierarchyAddBone(
    Space=GlobalSpace
    Name=@From_String_1.Result
    Transform=&RigUnit_GetTransform(bInitial=False, Item=@For_Each_1.Element).Transform
    Parent=RigVMDispatchNode(
     Values=((Type=Bone,Name="pelvis"),(Type=None,Name="None"),((),(),Type=None,Name="None"))
     Values.1=RigVMDispatchNode(Index=0, Array=@n3.Value) {template="DISPATCH_RigVMDispatch_ArrayGetAtIndex(in Array,in Index,out Element)"}.Element
     Values.2=RigVMDispatchNode(Index=1, Array=@n3.Value) {template="DISPATCH_RigVMDispatch_ArrayGetAtIndex(in Array,in Index,out Element)"}.Element
     Index=@For_Each_1.Index
    ) {template="DISPATCH_RigVMDispatch_SelectInt32(in Index,in Values,out Result)"}.Result
   )
   > n5 = RigVMDispatchNode(Other=((Type=None,Name="None")), Other.0=@n4.Item, Array=RigVMVariableNode() {variable=IK1_L_spawned}.Value) {template="DISPATCH_RigVMDispatch_ArrayAppend(io Array,in Other)"}
   > RigVMVariableNode(Value=@n5.Array) {variable=IK1_L_spawned}
   ComputePoleVector = RigVMFunctionReferenceNode(
    Bone A=(Type=Bone,Name="thigh_l")
    Bone B=(Type=Bone,Name="calf_l")
    Bone C=(Type=Bone,Name="ball_l")
    Draw Transform=true
    Initial=true
    OffsetFactor=-5
    Origin Scale=(X=0,Y=0,Z=0)
    ExecuteContext=@For_Each_1.Completed
   )
   > RigUnit_SetTransform = &RigUnit_SetTransform(Item=(Type=Bone,Name="PV_L"), Weight=1, bInitial=true, Value=@ComputePoleVector.Transform)
   n6 = RigVMVariableNode() {variable=IK1_R}
   For_Each_2 = RigVMDispatchNode(
    Array=((Type=None,Name="None"),(Type=None,Name="None"),(Type=None,Name="None"))
    Array.0=RigVMDispatchNode(Index=0, Array=@n6.Value) {template="DISPATCH_RigVMDispatch_ArrayGetAtIndex(in Array,in Index,out Element)"}.Element
    Array.1=RigVMDispatchNode(Index=1, Array=@n6.Value) {template="DISPATCH_RigVMDispatch_ArrayGetAtIndex(in Array,in Index,out Element)"}.Element
    Array.2=RigVMDispatchNode(Index=3, Array=@n6.Value) {template="DISPATCH_RigVMDispatch_ArrayGetAtIndex(in Array,in Index,out Element)"}.Element
    ExecuteContext=@RigVMFunction_Sequence_3.C
   ) {template="DISPATCH_RigVMDispatch_ArrayIterator(in Array,out Element,out Index,out Count,out Ratio)"}
   To_String_2 = RigVMDispatchNode(Value=@For_Each_2.Element.Name) {template="DISPATCH_RigDispatch_ToString(in Value,out Result)"}
   From_String_2 = RigVMDispatchNode(String=&RigVMFunction_StringJoin(Separator=_, Values=("IK1",""), Values.1=@To_String_2.Result).Result) {template="DISPATCH_RigDispatch_FromString(in String,out Result)"}
   n7 = RigVMVariableNode() {variable=IK1_R_spawned}
   > n8 = &RigUnit_HierarchyAddBone(
    Space=GlobalSpace
    Name=@From_String_2.Result
    Transform=&RigUnit_GetTransform(bInitial=False, Item=@For_Each_2.Element).Transform
    Parent=RigVMDispatchNode(
     Values=((Type=Bone,Name="pelvis"),(Type=None,Name="None"),((),(),Type=None,Name="None"))
     Values.1=RigVMDispatchNode(Index=0, Array=@n7.Value) {template="DISPATCH_RigVMDispatch_ArrayGetAtIndex(in Array,in Index,out Element)"}.Element
     Values.2=RigVMDispatchNode(Index=1, Array=@n7.Value) {template="DISPATCH_RigVMDispatch_ArrayGetAtIndex(in Array,in Index,out Element)"}.Element
     Index=@For_Each_2.Index
    ) {template="DISPATCH_RigVMDispatch_SelectInt32(in Index,in Values,out Result)"}.Result
   )
   > n9 = RigVMDispatchNode(Other=((Type=None,Name="None")), Other.0=@n8.Item, Array=RigVMVariableNode() {variable=IK1_R_spawned}.Value) {template="DISPATCH_RigVMDispatch_ArrayAppend(io Array,in Other)"}
   > RigVMVariableNode(Value=@n9.Array) {variable=IK1_R_spawned}
   ComputePoleVector_1 = RigVMFunctionReferenceNode(
    Bone A=(Type=Bone,Name="thigh_r")
    Bone B=(Type=Bone,Name="calf_r")
    Bone C=(Type=Bone,Name="foot_r")
    Draw Transform=true
    Initial=true
    OffsetFactor=20
    Origin Scale=(X=0,Y=0,Z=0)
    ExecuteContext=@For_Each_2.Completed
   )
   > &RigUnit_SetTransform(Item=(Type=Bone,Name="PV_R"), Weight=1, bInitial=true, Value=@ComputePoleVector_1.Transform)
   RigUnit_BeginExecution = &RigUnit_BeginExecution()
   /Script/FortniteGame.RigUnit_FortDeformRig(
    Bool=(bUpperArm=True,bElbow=True,bWrist=True,bGroin=True,bGlute=True,bKnee=True,bPec=True,bBicep=True,bLat=True)
    Curve="(bLUpperArmUpAlpha=False,bRUpperArmUpAlpha=False,bLUpperArmFrAlpha=False,bRUpperArmFrAlpha=False,bLUpperArmBkAlpha=False,bRUpperArmBkAlpha=False,bLWristUpAlpha=False,bRWristUpAlpha=False,bLWristDnAlpha=False,bRWristDnAlpha=False,bLKneeAlpha=true,bRKneeAlph ...(truncated)"
    Debug=(bUpperArmDebug=False,bWristDebug=False,bLatDebug=False,bGroinGluteDebug=False)
    Value="(UpperArmUp=(Rotation=(X=0.0,Y=-0.0,Z=0.0,W=1.0),Translation=(X=0.0,Y=0.0,Z=8.0),Scale3D=(X=0.70,Y=1.0,Z=1.0)),UpperArmFwdFr=(Rotation=(X=0.0,Y=-0.0,Z=0.0,W=1.0),Translation=(X=0.0,Y=10.0,Z=0.0),Scale3D=(X=0.50,Y=1.0,Z=1.0)),UpperArmFwdBk=(Rotation=(X=0.0, ...(truncated)"
    ExecutePin=@RigUnit_BeginExecution.ExecutePin
   ) {is=struct RigVMUnitNode}
   > &RigUnit_ParentConstraint(
    AdvancedSettings=(InterpolationType=Average,RotationOrderForFilter=XZY)
    Child=(Type=Bone,Name="phantom_thumb_02_r")
    Filter=(TranslationFilter=(bX=false,bY=false,bZ=false),RotationFilter=(bX=True,bY=True,bZ=True),ScaleFilter=(bX=True,bY=True,bZ=True))
    Parents=((Item=(Type=Bone,Name="thumb_02_r"),Weight=1))
    Weight=1
   )
   > &RigUnit_ParentConstraint(
    AdvancedSettings=(InterpolationType=Average,RotationOrderForFilter=XZY)
    Child=(Type=Bone,Name="phantom_thumb_03_r")
    Filter=(TranslationFilter=(bX=false,bY=false,bZ=false),RotationFilter=(bX=True,bY=True,bZ=True),ScaleFilter=(bX=True,bY=True,bZ=True))
    Parents=((Item=(Type=Bone,Name="thumb_03_r"),Weight=1))
    Weight=1
   )
   > &RigUnit_ParentConstraint(
    AdvancedSettings=(InterpolationType=Average,RotationOrderForFilter=XZY)
    Child=(Type=Bone,Name="phantom_index_01_r")
    Filter=(TranslationFilter=(bX=false,bY=false,bZ=false),RotationFilter=(bX=True,bY=True,bZ=True),ScaleFilter=(bX=True,bY=True,bZ=True))
    Parents=((Item=(Type=Bone,Name="index_01_r"),Weight=1))
    Weight=1
   )
   > &RigUnit_ParentConstraint(
    AdvancedSettings=(InterpolationType=Average,RotationOrderForFilter=XZY)
    Child=(Type=Bone,Name="phantom_index_02_r")
    Filter=(TranslationFilter=(bX=false,bY=false,bZ=false),RotationFilter=(bX=True,bY=True,bZ=True),ScaleFilter=(bX=True,bY=True,bZ=True))
    Parents=((Item=(Type=Bone,Name="index_02_r"),Weight=1))
    Weight=1
   )
   > &RigUnit_ParentConstraint(
    AdvancedSettings=(InterpolationType=Average,RotationOrderForFilter=XZY)
    Child=(Type=Bone,Name="phantom_index_03_r")
    Filter=(TranslationFilter=(bX=false,bY=false,bZ=false),RotationFilter=(bX=True,bY=True,bZ=True),ScaleFilter=(bX=True,bY=True,bZ=True))
    Parents=((Item=(Type=Bone,Name="index_03_r"),Weight=1))
    Weight=1
   )
   > &RigUnit_ParentConstraint(
    AdvancedSettings=(InterpolationType=Average,RotationOrderForFilter=XZY)
    Child=(Type=Bone,Name="phantom_middle_01_r")
    Filter=(TranslationFilter=(bX=false,bY=false,bZ=false),RotationFilter=(bX=True,bY=True,bZ=True),ScaleFilter=(bX=True,bY=True,bZ=True))
    Parents=((Item=(Type=Bone,Name="middle_01_r"),Weight=1))
    Weight=1
   )
   > &RigUnit_ParentConstraint(
    AdvancedSettings=(InterpolationType=Average,RotationOrderForFilter=XZY)
    Child=(Type=Bone,Name="phantom_middle_02_r")
    Filter=(TranslationFilter=(bX=false,bY=false,bZ=false),RotationFilter=(bX=True,bY=True,bZ=True),ScaleFilter=(bX=True,bY=True,bZ=True))
    Parents=((Item=(Type=Bone,Name="middle_02_r"),Weight=1))
    Weight=1
   )
   > &RigUnit_ParentConstraint(
    AdvancedSettings=(InterpolationType=Average,RotationOrderForFilter=XZY)
    Child=(Type=Bone,Name="phantom_middle_03_r")
    Filter=(TranslationFilter=(bX=false,bY=false,bZ=false),RotationFilter=(bX=True,bY=True,bZ=True),ScaleFilter=(bX=True,bY=True,bZ=True))
    Parents=((Item=(Type=Bone,Name="middle_03_r"),Weight=1))
    Weight=1
   )
   > &RigUnit_ParentConstraint(
    AdvancedSettings=(InterpolationType=Average,RotationOrderForFilter=XZY)
    Child=(Type=Bone,Name="phantom_ring_01_r")
    Filter=(TranslationFilter=(bX=false,bY=false,bZ=false),RotationFilter=(bX=True,bY=True,bZ=True),ScaleFilter=(bX=True,bY=True,bZ=True))
    Parents=((Item=(Type=Bone,Name="ring_01_r"),Weight=1))
    Weight=1
   )
   > &RigUnit_ParentConstraint(
    AdvancedSettings=(InterpolationType=Average,RotationOrderForFilter=XZY)
    Child=(Type=Bone,Name="phantom_ring_02_r")
    Filter=(TranslationFilter=(bX=false,bY=false,bZ=false),RotationFilter=(bX=True,bY=True,bZ=True),ScaleFilter=(bX=True,bY=True,bZ=True))
    Parents=((Item=(Type=Bone,Name="ring_02_r"),Weight=1))
    Weight=1
   )
   > &RigUnit_ParentConstraint(
    AdvancedSettings=(InterpolationType=Average,RotationOrderForFilter=XZY)
    Child=(Type=Bone,Name="phantom_ring_03_r")
    Filter=(TranslationFilter=(bX=false,bY=false,bZ=false),RotationFilter=(bX=True,bY=True,bZ=True),ScaleFilter=(bX=True,bY=True,bZ=True))
    Parents=((Item=(Type=Bone,Name="ring_03_r"),Weight=1))
    Weight=1
   )
   > &RigUnit_ParentConstraint(
    AdvancedSettings=(InterpolationType=Average,RotationOrderForFilter=XZY)
    Child=(Type=Bone,Name="phantom_thumb_02_l")
    Filter=(TranslationFilter=(bX=false,bY=false,bZ=false),RotationFilter=(bX=True,bY=True,bZ=True),ScaleFilter=(bX=True,bY=True,bZ=True))
    Parents=((Item=(Type=Bone,Name="thumb_02_l"),Weight=1))
    Weight=1
   )
   > &RigUnit_ParentConstraint(
    AdvancedSettings=(InterpolationType=Average,RotationOrderForFilter=XZY)
    Child=(Type=Bone,Name="phantom_thumb_03_l")
    Filter=(TranslationFilter=(bX=false,bY=false,bZ=false),RotationFilter=(bX=True,bY=True,bZ=True),ScaleFilter=(bX=True,bY=True,bZ=True))
    Parents=((Item=(Type=Bone,Name="thumb_03_l"),Weight=1))
    Weight=1
   )
   > &RigUnit_ParentConstraint(
    AdvancedSettings=(InterpolationType=Average,RotationOrderForFilter=XZY)
    Child=(Type=Bone,Name="phantom_index_01_l")
    Filter=(TranslationFilter=(bX=false,bY=false,bZ=false),RotationFilter=(bX=True,bY=True,bZ=True),ScaleFilter=(bX=True,bY=True,bZ=True))
    Parents=((Item=(Type=Bone,Name="index_01_l"),Weight=1))
    Weight=1
   )
   > &RigUnit_ParentConstraint(
    AdvancedSettings=(InterpolationType=Average,RotationOrderForFilter=XZY)
    Child=(Type=Bone,Name="phantom_index_02_l")
    Filter=(TranslationFilter=(bX=false,bY=false,bZ=false),RotationFilter=(bX=True,bY=True,bZ=True),ScaleFilter=(bX=True,bY=True,bZ=True))
    Parents=((Item=(Type=Bone,Name="index_02_l"),Weight=1))
    Weight=1
   )
   > &RigUnit_ParentConstraint(
    AdvancedSettings=(InterpolationType=Average,RotationOrderForFilter=XZY)
    Child=(Type=Bone,Name="phantom_index_03_l")
    Filter=(TranslationFilter=(bX=false,bY=false,bZ=false),RotationFilter=(bX=True,bY=True,bZ=True),ScaleFilter=(bX=True,bY=True,bZ=True))
    Parents=((Item=(Type=Bone,Name="index_03_l"),Weight=1))
    Weight=1
   )
   > &RigUnit_ParentConstraint(
    AdvancedSettings=(InterpolationType=Average,RotationOrderForFilter=XZY)
    Child=(Type=Bone,Name="phantom_middle_01_l")
    Filter=(TranslationFilter=(bX=false,bY=false,bZ=false),RotationFilter=(bX=True,bY=True,bZ=True),ScaleFilter=(bX=True,bY=True,bZ=True))
    Parents=((Item=(Type=Bone,Name="middle_01_l"),Weight=1))
    Weight=1
   )
   > &RigUnit_ParentConstraint(
    AdvancedSettings=(InterpolationType=Average,RotationOrderForFilter=XZY)
    Child=(Type=Bone,Name="phantom_middle_02_l")
    Filter=(TranslationFilter=(bX=false,bY=false,bZ=false),RotationFilter=(bX=True,bY=True,bZ=True),ScaleFilter=(bX=True,bY=True,bZ=True))
    Parents=((Item=(Type=Bone,Name="middle_02_l"),Weight=1))
    Weight=1
   )
   > &RigUnit_ParentConstraint(
    AdvancedSettings=(InterpolationType=Average,RotationOrderForFilter=XZY)
    Child=(Type=Bone,Name="phantom_middle_03_l")
    Filter=(TranslationFilter=(bX=false,bY=false,bZ=false),RotationFilter=(bX=True,bY=True,bZ=True),ScaleFilter=(bX=True,bY=True,bZ=True))
    Parents=((Item=(Type=Bone,Name="middle_03_l"),Weight=1))
    Weight=1
   )
   > &RigUnit_ParentConstraint(
    AdvancedSettings=(InterpolationType=Average,RotationOrderForFilter=XZY)
    Child=(Type=Bone,Name="phantom_ring_01_l")
    Filter=(TranslationFilter=(bX=false,bY=false,bZ=false),RotationFilter=(bX=True,bY=True,bZ=True),ScaleFilter=(bX=True,bY=True,bZ=True))
    Parents=((Item=(Type=Bone,Name="ring_01_l"),Weight=1))
    Weight=1
   )
   > &RigUnit_ParentConstraint(
    AdvancedSettings=(InterpolationType=Average,RotationOrderForFilter=XZY)
    Child=(Type=Bone,Name="phantom_ring_02_l")
    Filter=(TranslationFilter=(bX=false,bY=false,bZ=false),RotationFilter=(bX=True,bY=True,bZ=True),ScaleFilter=(bX=True,bY=True,bZ=True))
    Parents=((Item=(Type=Bone,Name="ring_02_l"),Weight=1))
    Weight=1
   )
   > &RigUnit_ParentConstraint(
    AdvancedSettings=(InterpolationType=Average,RotationOrderForFilter=XZY)
    Child=(Type=Bone,Name="phantom_ring_03_l")
    Filter=(TranslationFilter=(bX=false,bY=false,bZ=false),RotationFilter=(bX=True,bY=True,bZ=True),ScaleFilter=(bX=True,bY=True,bZ=True))
    Parents=((Item=(Type=Bone,Name="ring_03_l"),Weight=1))
    Weight=1
   )
   > &RigUnit_ModifyTransforms(
    ItemToModify="((Item=(Type=Bone,Name=\"R_tentacle_01_M\"),Transform=(Rotation=(X=0,Y=0.0436,Z=0,W=0.999),Translation=(X=0,Y=0,Z=0),Scale3D=(X=1,Y=1,Z=1))),(Item=(Type=Bone,Name=\"R_tentacle_01_L1\"),Transform=(Rotation=(X=0.0227,Y=0.1285,Z=-0.1722,W=0.9764),Translation=(X=0 ...(truncated)"
    Weight=1
    WeightMinimum=0
   )
   > &RigUnit_ModifyTransforms(
    ItemToModify="((Item=(Type=Bone,Name=\"Pelvis_coorskin\"),Transform=(Rotation=(X=0,Y=0,Z=0,W=1),Translation=(X=0,Y=0,Z=-1),Scale3D=(X=1,Y=1,Z=1))),(Item=(Type=Bone,Name=\"pelvic_bone\"),Transform=(Rotation=(X=0,Y=0,Z=0,W=1),Translation=(X=0,Y=0,Z=2),Scale3D=(X=1,Y=0.7,Z=1)) ...(truncated)"
    Weight=1
    WeightMinimum=0
   )
   > RigVMFunction_Sequence = RigVMAggregateNode()
   RigVMFunction_ControlFlowBranch = &RigVMFunction_ControlFlowBranch(ExecuteContext=@RigVMFunction_Sequence.B, Condition=RigVMVariableNode() {variable=isAim}.Value)
   &RigUnit_ModifyTransforms(
    ItemToModify="((Item=(Type=Bone,Name=\"R_tentacle_01_R1\"),Transform=(Rotation=(X=0,Y=0,Z=-0.2164,W=0.9763),Translation=(X=0,Y=0,Z=0),Scale3D=(X=1,Y=1,Z=1))),(Item=(Type=Bone,Name=\"R_tentacle_01_R2\"),Transform=(Rotation=(X=0,Y=0,Z=-0.1736,W=0.9848),Translation=(X=0,Y=0,Z= ...(truncated)"
    WeightMinimum=0
    ExecutePin=@RigVMFunction_ControlFlowBranch.Completed
    Weight=RigVMVariableNode() {variable=Aiming}.Value
   )
   RigVMFunction_MathQuaternionToEuler = &RigVMFunction_MathQuaternionToEuler(
    RotationOrder=ZYX
    Value=&RigUnit_GetRelativeTransformForItem(Child=(Type=Bone,Name="head"), Parent=(Type=Bone,Name="spine_05"), bChildInitial=False, bParentInitial=False).RelativeTransform.Rotation
   )
   > &RigUnit_ModifyTransforms(
    ItemToModify="((Item=(Type=Bone,Name=\"R_tentacle_01_R1\"),Transform=(Rotation=(X=0.0973,Y=0.378,Z=0.2159,W=-0.895),Translation=(X=0,Y=0,Z=0),Scale3D=(X=1,Y=1,Z=1))),(Item=(Type=Bone,Name=\"R_tentacle_01_R2\"),Transform=(Rotation=(X=0.039,Y=-0.2313,Z=-0.1711,W=0.9569),Trans ...(truncated)"
    WeightMinimum=0
    Weight=&RigVMFunction_MathDoubleMul(
     B=&RigVMFunction_MathDoubleRemap(SourceMaximum=-42, SourceMinimum=-20, TargetMaximum=1, TargetMinimum=0, Value=@RigVMFunction_MathQuaternionToEuler.Result.Z).Result
     A=RigVMVariableNode() {variable=Aiming}.Value
    ).Result
   )
   n10 = &RigUnit_GetRelativeTransformForItem(Child=(Type=Bone,Name="head"), Parent=(Type=Bone,Name="spine_05"), bChildInitial=False, bParentInitial=False)
   > n11 = &RigUnit_ModifyTransforms(
    ItemToModify="((Item=(Type=Bone,Name=\"R_tentacle_01_L2\"),Transform=(Rotation=(X=0,Y=-0.1736,Z=0,W=0.9848),Translation=(X=0,Y=0,Z=0),Scale3D=(X=1,Y=1,Z=1))),(Item=(Type=Bone,Name=\"R_tentacle_01_R2\"),Transform=(Rotation=(X=0,Y=-0.1736,Z=0,W=0.9848),Translation=(X=0,Y=0,Z= ...(truncated)"
    WeightMinimum=0
    Weight=&RigVMFunction_MathDoubleRemap(
     SourceMaximum=-42
     SourceMinimum=-20
     TargetMaximum=1
     TargetMinimum=0
     Value=&RigVMFunction_MathQuaternionToEuler(RotationOrder=ZYX, Value=@n10.RelativeTransform.Rotation).Result.Z
    ).Result
   )
   n12 = &RigVMFunction_ControlFlowBranch(ExecuteContext=@n11.ExecutePin, Condition=RigVMVariableNode() {variable=skyDiving}.Value)
   &RigUnit_ModifyTransforms(
    ItemToModify="((Item=(Type=Bone,Name=\"R_tentacle_01_M\"),Transform=(Rotation=(X=0,Y=-0.1736,Z=0,W=0.9848),Translation=(X=0,Y=0,Z=0),Scale3D=(X=1,Y=1,Z=1))),(Item=(Type=Bone,Name=\"R_tentacle_01_L1\"),Transform=(Rotation=(X=0,Y=-0.1736,Z=0,W=0.9848),Translation=(X=0,Y=0,Z=0 ...(truncated)"
    WeightMinimum=0
    Weight=&RigVMFunction_MathDoubleRemap(SourceMaximum=5000, SourceMinimum=2000, TargetMaximum=1, TargetMinimum=0, Value=RigVMVariableNode() {variable=GoingUP}.Value).Result
    ExecutePin=@n12.True
   )
   /Script/ControlRig.RigUnit_ChainHarmonicsPerItem(
    ChainRoot=(Type=Bone,Name="R_tentacle_01_M")
    DrawWorldOffset=(Rotation=(X=0,Y=0,Z=0,W=1),Translation=(X=0,Y=0,Z=0),Scale3D=(X=1,Y=1,Z=1))
    Pendulum=(bEnabled=false,PendulumStiffness=50,PendulumGravity=(X=-900,Y=-500,Z=0),PendulumBlend=1,PendulumDrag=20,PendulumMinimum=0,PendulumMaximum=0.8,PendulumEase=SineEaseInOut,UnwindAxis=(X=0,Y=0,Z=0),UnwindMinimum=0,UnwindMaximum=0)
    Reach=(bEnabled=false,ReachTarget=(X=0,Y=0,Z=0),ReachAxis=(X=1,Y=1,Z=1),ReachMinimum=0,ReachMaximum=0,ReachEase=SineEaseInOut)
    Speed=(X=0,Y=-3,Z=0)
    Wave=(bEnabled=True,WaveFrequency=(X=0,Y=10,Z=0),WaveAmplitude=(X=0,Y=-4,Z=0),WaveOffset=(X=0,Y=2,Z=0),WaveNoise=(X=0,Y=0,Z=0),WaveMinimum=0.1,WaveMaximum=1,WaveEase=SineEaseInOut)
    WaveCurve="(EditorCurveData=(Keys=((InterpMode=RCIM_Linear,TangentMode=RCTM_Auto,TangentWeightMode=RCTWM_WeightedNone,Time=0,Value=0.2,ArriveTangent=0,ArriveTangentWeight=0,LeaveTangent=0,LeaveTangentWeight=0),(InterpMode=RCIM_Linear,TangentMode=RCTM_Auto,TangentWeig ...(truncated)"
    bDrawDebug=True
    ExecutePin=@RigVMFunction_ControlFlowBranch.False
   ) {is=struct RigVMUnitNode}
   > &RigUnit_ChainHarmonicsPerItem(
    ChainRoot=(Type=Bone,Name="R_tentacle_01_L1")
    DrawWorldOffset=(Rotation=(X=0,Y=0,Z=0,W=1),Translation=(X=0,Y=0,Z=0),Scale3D=(X=1,Y=1,Z=1))
    Pendulum=(bEnabled=false,PendulumStiffness=50,PendulumGravity=(X=1000,Y=0,Z=0),PendulumBlend=1,PendulumDrag=0.98,PendulumMinimum=0,PendulumMaximum=0.9,PendulumEase=SineEaseInOut,UnwindAxis=(X=0,Y=0,Z=0),UnwindMinimum=0,UnwindMaximum=0)
    Reach=(bEnabled=false,ReachTarget=(X=1,Y=10,Z=1),ReachAxis=(X=1,Y=1,Z=1),ReachMinimum=10,ReachMaximum=20,ReachEase=SineEaseInOut)
    Speed=(X=0,Y=-3,Z=0)
    Wave=(bEnabled=True,WaveFrequency=(X=0,Y=10,Z=0),WaveAmplitude=(X=0,Y=3,Z=0),WaveOffset=(X=0,Y=0,Z=0),WaveNoise=(X=0,Y=0,Z=0),WaveMinimum=0.1,WaveMaximum=1,WaveEase=SineEaseInOut)
    WaveCurve="(EditorCurveData=(Keys=((InterpMode=RCIM_Linear,TangentMode=RCTM_Auto,TangentWeightMode=RCTWM_WeightedNone,Time=0,Value=0.2,ArriveTangent=0,ArriveTangentWeight=0,LeaveTangent=0,LeaveTangentWeight=0),(InterpMode=RCIM_Linear,TangentMode=RCTM_Auto,TangentWeig ...(truncated)"
    bDrawDebug=True
   )
   > &RigUnit_ChainHarmonicsPerItem(
    ChainRoot=(Type=Bone,Name="R_tentacle_01_R1")
    DrawWorldOffset=(Rotation=(X=0,Y=0,Z=0,W=1),Translation=(X=0,Y=0,Z=0),Scale3D=(X=1,Y=1,Z=1))
    Pendulum=(bEnabled=false,PendulumStiffness=50,PendulumGravity=(X=1000,Y=0,Z=0),PendulumBlend=1,PendulumDrag=0.98,PendulumMinimum=0,PendulumMaximum=0.9,PendulumEase=SineEaseInOut,UnwindAxis=(X=0,Y=0,Z=0),UnwindMinimum=0,UnwindMaximum=0)
    Reach=(bEnabled=false,ReachTarget=(X=1,Y=10,Z=1),ReachAxis=(X=1,Y=1,Z=1),ReachMinimum=10,ReachMaximum=20,ReachEase=SineEaseInOut)
    Speed=(X=0,Y=-3,Z=0)
    Wave=(bEnabled=True,WaveFrequency=(X=0,Y=10,Z=0),WaveAmplitude=(X=0,Y=3,Z=0),WaveOffset=(X=0,Y=0,Z=0),WaveNoise=(X=0,Y=0,Z=0),WaveMinimum=0.1,WaveMaximum=1,WaveEase=SineEaseInOut)
    WaveCurve="(EditorCurveData=(Keys=((InterpMode=RCIM_Linear,TangentMode=RCTM_Auto,TangentWeightMode=RCTWM_WeightedNone,Time=0,Value=0.2,ArriveTangent=0,ArriveTangentWeight=0,LeaveTangent=0,LeaveTangentWeight=0),(InterpMode=RCIM_Linear,TangentMode=RCTM_Auto,TangentWeig ...(truncated)"
    bDrawDebug=True
   )
   > &RigUnit_ChainHarmonicsPerItem(
    ChainRoot=(Type=Bone,Name="R_tentacle_01_L2")
    DrawWorldOffset=(Rotation=(X=0,Y=0,Z=0,W=1),Translation=(X=0,Y=0,Z=0),Scale3D=(X=1,Y=1,Z=1))
    Pendulum=(bEnabled=false,PendulumStiffness=100,PendulumGravity=(X=-500,Y=900,Z=0),PendulumBlend=1,PendulumDrag=0.9,PendulumMinimum=0.2,PendulumMaximum=0.9,PendulumEase=SineEaseInOut,UnwindAxis=(X=0,Y=0,Z=0),UnwindMinimum=0,UnwindMaximum=0)
    Reach=(bEnabled=false,ReachTarget=(X=1,Y=10,Z=1),ReachAxis=(X=-1,Y=1,Z=-1),ReachMinimum=10,ReachMaximum=20,ReachEase=SineEaseInOut)
    Speed=(X=0,Y=-3,Z=0)
    Wave=(bEnabled=True,WaveFrequency=(X=0,Y=10,Z=0),WaveAmplitude=(X=0,Y=-3,Z=0),WaveOffset=(X=0,Y=-1,Z=0),WaveNoise=(X=0,Y=0,Z=0),WaveMinimum=0.1,WaveMaximum=1,WaveEase=SineEaseInOut)
    WaveCurve="(EditorCurveData=(Keys=((InterpMode=RCIM_Linear,TangentMode=RCTM_Auto,TangentWeightMode=RCTWM_WeightedNone,Time=0,Value=0.2,ArriveTangent=0,ArriveTangentWeight=0,LeaveTangent=0,LeaveTangentWeight=0),(InterpMode=RCIM_Linear,TangentMode=RCTM_Auto,TangentWeig ...(truncated)"
    bDrawDebug=True
   )
   > &RigUnit_ChainHarmonicsPerItem(
    ChainRoot=(Type=Bone,Name="R_tentacle_01_R2")
    DrawWorldOffset=(Rotation=(X=0,Y=0,Z=0,W=1),Translation=(X=0,Y=0,Z=0),Scale3D=(X=1,Y=1,Z=1))
    Pendulum=(bEnabled=false,PendulumStiffness=100,PendulumGravity=(X=-500,Y=900,Z=0),PendulumBlend=1,PendulumDrag=0.9,PendulumMinimum=0.2,PendulumMaximum=0.9,PendulumEase=SineEaseInOut,UnwindAxis=(X=0,Y=0,Z=0),UnwindMinimum=0,UnwindMaximum=0)
    Reach=(bEnabled=false,ReachTarget=(X=1,Y=10,Z=1),ReachAxis=(X=-1,Y=1,Z=-1),ReachMinimum=10,ReachMaximum=20,ReachEase=SineEaseInOut)
    Speed=(X=0,Y=-3,Z=0)
    Wave=(bEnabled=True,WaveFrequency=(X=0,Y=10,Z=0),WaveAmplitude=(X=0,Y=-3,Z=0),WaveOffset=(X=0,Y=-1,Z=0),WaveNoise=(X=0,Y=0,Z=0),WaveMinimum=0.1,WaveMaximum=1,WaveEase=SineEaseInOut)
    WaveCurve="(EditorCurveData=(Keys=((InterpMode=RCIM_Linear,TangentMode=RCTM_Auto,TangentWeightMode=RCTWM_WeightedNone,Time=0,Value=0.2,ArriveTangent=0,ArriveTangentWeight=0,LeaveTangent=0,LeaveTangentWeight=0),(InterpMode=RCIM_Linear,TangentMode=RCTM_Auto,TangentWeig ...(truncated)"
    bDrawDebug=True
   )
   &RigUnit_ChainHarmonicsPerItem(
    ChainRoot=(Type=Bone,Name="R_tentacle_01_M")
    DrawWorldOffset=(Rotation=(X=0,Y=0,Z=0,W=1),Translation=(X=0,Y=0,Z=0),Scale3D=(X=1,Y=1,Z=1))
    Pendulum=(bEnabled=false,PendulumStiffness=50,PendulumGravity=(X=-900,Y=-500,Z=0),PendulumBlend=1,PendulumDrag=20,PendulumMinimum=0,PendulumMaximum=0.8,PendulumEase=SineEaseInOut,UnwindAxis=(X=0,Y=0,Z=0),UnwindMinimum=0,UnwindMaximum=0)
    Reach=(bEnabled=false,ReachTarget=(X=0,Y=0,Z=0),ReachAxis=(X=1,Y=1,Z=1),ReachMinimum=0,ReachMaximum=0,ReachEase=SineEaseInOut)
    Speed=(X=0,Y=-2,Z=0)
    Wave=(bEnabled=True,WaveFrequency=(X=0,Y=10,Z=0),WaveAmplitude=(X=0,Y=-4,Z=0),WaveOffset=(X=0,Y=2,Z=0),WaveNoise=(X=0,Y=0,Z=0),WaveMinimum=0.1,WaveMaximum=1,WaveEase=SineEaseInOut)
    WaveCurve="(EditorCurveData=(Keys=((InterpMode=RCIM_Linear,TangentMode=RCTM_Auto,TangentWeightMode=RCTWM_WeightedNone,Time=0,Value=0.2,ArriveTangent=0,ArriveTangentWeight=0,LeaveTangent=0,LeaveTangentWeight=0),(InterpMode=RCIM_Linear,TangentMode=RCTM_Auto,TangentWeig ...(truncated)"
    bDrawDebug=True
    ExecutePin=@RigVMFunction_ControlFlowBranch.True
   )
   > &RigUnit_ChainHarmonicsPerItem(
    ChainRoot=(Type=Bone,Name="R_tentacle_01_L1")
    DrawWorldOffset=(Rotation=(X=0,Y=0,Z=0,W=1),Translation=(X=0,Y=0,Z=0),Scale3D=(X=1,Y=1,Z=1))
    Pendulum=(bEnabled=false,PendulumStiffness=50,PendulumGravity=(X=1000,Y=0,Z=0),PendulumBlend=1,PendulumDrag=0.98,PendulumMinimum=0,PendulumMaximum=0.9,PendulumEase=SineEaseInOut,UnwindAxis=(X=0,Y=0,Z=0),UnwindMinimum=0,UnwindMaximum=0)
    Reach=(bEnabled=false,ReachTarget=(X=1,Y=10,Z=1),ReachAxis=(X=1,Y=1,Z=1),ReachMinimum=10,ReachMaximum=20,ReachEase=SineEaseInOut)
    Speed=(X=0,Y=-2,Z=0)
    Wave=(bEnabled=True,WaveFrequency=(X=0,Y=10,Z=0),WaveAmplitude=(X=0,Y=3,Z=0),WaveOffset=(X=0,Y=0,Z=0),WaveNoise=(X=0,Y=0,Z=0),WaveMinimum=0.1,WaveMaximum=1,WaveEase=SineEaseInOut)
    WaveCurve="(EditorCurveData=(Keys=((InterpMode=RCIM_Linear,TangentMode=RCTM_Auto,TangentWeightMode=RCTWM_WeightedNone,Time=0,Value=0.2,ArriveTangent=0,ArriveTangentWeight=0,LeaveTangent=0,LeaveTangentWeight=0),(InterpMode=RCIM_Linear,TangentMode=RCTM_Auto,TangentWeig ...(truncated)"
    bDrawDebug=True
   )
   > &RigUnit_ChainHarmonicsPerItem(
    ChainRoot=(Type=Bone,Name="R_tentacle_01_R1")
    DrawWorldOffset=(Rotation=(X=0,Y=0,Z=0,W=1),Translation=(X=0,Y=0,Z=0),Scale3D=(X=1,Y=1,Z=1))
    Pendulum=(bEnabled=false,PendulumStiffness=50,PendulumGravity=(X=1000,Y=0,Z=0),PendulumBlend=1,PendulumDrag=0.98,PendulumMinimum=0,PendulumMaximum=0.9,PendulumEase=SineEaseInOut,UnwindAxis=(X=0,Y=0,Z=0),UnwindMinimum=0,UnwindMaximum=0)
    Reach=(bEnabled=false,ReachTarget=(X=1,Y=10,Z=1),ReachAxis=(X=1,Y=1,Z=1),ReachMinimum=10,ReachMaximum=20,ReachEase=SineEaseInOut)
    Speed=(X=0,Y=-2,Z=0)
    Wave=(bEnabled=True,WaveFrequency=(X=0,Y=10,Z=0),WaveAmplitude=(X=0,Y=3,Z=0),WaveOffset=(X=0,Y=0,Z=0),WaveNoise=(X=0,Y=0,Z=0),WaveMinimum=0.1,WaveMaximum=1,WaveEase=SineEaseInOut)
    WaveCurve="(EditorCurveData=(Keys=((InterpMode=RCIM_Linear,TangentMode=RCTM_Auto,TangentWeightMode=RCTWM_WeightedNone,Time=0,Value=0.2,ArriveTangent=0,ArriveTangentWeight=0,LeaveTangent=0,LeaveTangentWeight=0),(InterpMode=RCIM_Linear,TangentMode=RCTM_Auto,TangentWeig ...(truncated)"
    bDrawDebug=True
   )
   > &RigUnit_ChainHarmonicsPerItem(
    ChainRoot=(Type=Bone,Name="R_tentacle_01_L2")
    DrawWorldOffset=(Rotation=(X=0,Y=0,Z=0,W=1),Translation=(X=0,Y=0,Z=0),Scale3D=(X=1,Y=1,Z=1))
    Pendulum=(bEnabled=false,PendulumStiffness=100,PendulumGravity=(X=-500,Y=900,Z=0),PendulumBlend=1,PendulumDrag=0.9,PendulumMinimum=0.2,PendulumMaximum=0.9,PendulumEase=SineEaseInOut,UnwindAxis=(X=0,Y=0,Z=0),UnwindMinimum=0,UnwindMaximum=0)
    Reach=(bEnabled=false,ReachTarget=(X=1,Y=10,Z=1),ReachAxis=(X=-1,Y=1,Z=-1),ReachMinimum=10,ReachMaximum=20,ReachEase=SineEaseInOut)
    Speed=(X=0,Y=-2,Z=0)
    Wave=(bEnabled=True,WaveFrequency=(X=0,Y=10,Z=0),WaveAmplitude=(X=0,Y=-3,Z=0),WaveOffset=(X=0,Y=-1,Z=0),WaveNoise=(X=0,Y=0,Z=0),WaveMinimum=0.1,WaveMaximum=1,WaveEase=SineEaseInOut)
    WaveCurve="(EditorCurveData=(Keys=((InterpMode=RCIM_Linear,TangentMode=RCTM_Auto,TangentWeightMode=RCTWM_WeightedNone,Time=0,Value=0.2,ArriveTangent=0,ArriveTangentWeight=0,LeaveTangent=0,LeaveTangentWeight=0),(InterpMode=RCIM_Linear,TangentMode=RCTM_Auto,TangentWeig ...(truncated)"
    bDrawDebug=True
   )
   > &RigUnit_ChainHarmonicsPerItem(
    ChainRoot=(Type=Bone,Name="R_tentacle_01_R2")
    DrawWorldOffset=(Rotation=(X=0,Y=0,Z=0,W=1),Translation=(X=0,Y=0,Z=0),Scale3D=(X=1,Y=1,Z=1))
    Pendulum=(bEnabled=false,PendulumStiffness=100,PendulumGravity=(X=-500,Y=900,Z=0),PendulumBlend=1,PendulumDrag=0.9,PendulumMinimum=0.2,PendulumMaximum=0.9,PendulumEase=SineEaseInOut,UnwindAxis=(X=0,Y=0,Z=0),UnwindMinimum=0,UnwindMaximum=0)
    Reach=(bEnabled=false,ReachTarget=(X=1,Y=10,Z=1),ReachAxis=(X=-1,Y=1,Z=-1),ReachMinimum=10,ReachMaximum=20,ReachEase=SineEaseInOut)
    Speed=(X=0,Y=-2,Z=0)
    Wave=(bEnabled=True,WaveFrequency=(X=0,Y=10,Z=0),WaveAmplitude=(X=0,Y=-3,Z=0),WaveOffset=(X=0,Y=-1,Z=0),WaveNoise=(X=0,Y=0,Z=0),WaveMinimum=0.1,WaveMaximum=1,WaveEase=SineEaseInOut)
    WaveCurve="(EditorCurveData=(Keys=((InterpMode=RCIM_Linear,TangentMode=RCTM_Auto,TangentWeightMode=RCTWM_WeightedNone,Time=0,Value=0.2,ArriveTangent=0,ArriveTangentWeight=0,LeaveTangent=0,LeaveTangentWeight=0),(InterpMode=RCIM_Linear,TangentMode=RCTM_Auto,TangentWeig ...(truncated)"
    bDrawDebug=True
   )
   n13 = RigVMVariableNode() {variable=isPlatEmote}
   n14 = &RigVMFunction_AlphaInterp(
    Bias=0.1
    InterpSpeedDecreasing=10
    InterpSpeedIncreasing=20
    Scale=1
    bClampResult=true
    bMapRange=false
    Value=RigVMDispatchNode(False=1, True=0, Condition=@n13.Value) {template="DISPATCH_RigVMDispatch_If(in Condition,in True,in False,out Result)"}.Result
   )
   RigVMVariableNode(Value=@n14.Result, ExecuteContext=@RigVMFunction_Sequence.A) {variable=EndInterpEmote}
   > RigVMFunction_Sequence_1 = RigVMAggregateNode()
   &RigUnit_ParentConstraint(
    AdvancedSettings=(InterpolationType=Average,RotationOrderForFilter=XZY)
    Child=(Type=Bone,Name="ctrl_ball_l")
    Parents=((Item=(Type=Bone,Name="ball_l"),Weight=1))
    Weight=1
    bMaintainOffset=true
    ExecutePin=@RigVMFunction_Sequence_1.C
   )
   > n15 = &RigUnit_ParentConstraint(
    AdvancedSettings=(InterpolationType=Average,RotationOrderForFilter=XZY)
    Child=(Type=Bone,Name="ctrl_ball_r")
    Parents=((Item=(Type=Bone,Name="ball_r"),Weight=1))
    Weight=1
   )
   n16 = &RigVMFunction_ControlFlowBranch(ExecuteContext=@n15.ExecutePin, Condition=RigVMVariableNode() {variable=isCustomBoots}.Value)
   RigVMFunction_Sequence_2 = RigVMAggregateNode(ExecuteContext=@n16.False)
   &RigUnit_ParentConstraint(
    AdvancedSettings=(InterpolationType=Average,RotationOrderForFilter=XZY)
    Child=(Type=Bone,Name="PV_R")
    Parents=((Item=(Type=Bone,Name="thigh_r"),Weight=1))
    Weight=1
    ExecutePin=@RigVMFunction_Sequence_2.B
   )
   n17 = RigVMVariableNode() {variable=IK1_R_spawned}
   > IKTwoBone_2 = RigVMFunctionReferenceNode(
    Enable Stretch=true
    End Ctrl=(Type=Bone,Name="ctrl_ball_r")
    Line Thickness=0.25
    Pole Vector Node=(Type=Bone,Name="PV_R")
    Primary Axis=(X=1,Y=0,Z=0)
    Secondary Axis=(X=0,Y=-1,Z=0)
    Stretch Maximum Ratio=1.25
    Stretch Start Ratio=0.9
    debug=true
    scale=(X=0,Y=0,Z=0)
    Bone A=RigVMDispatchNode(Index=0, Array=@n17.Value) {template="DISPATCH_RigVMDispatch_ArrayGetAtIndex(in Array,in Index,out Element)"}.Element
    Bone B=RigVMDispatchNode(Index=1, Array=@n17.Value) {template="DISPATCH_RigVMDispatch_ArrayGetAtIndex(in Array,in Index,out Element)"}.Element
    Bone C=RigVMDispatchNode(Index=2, Array=@n17.Value) {template="DISPATCH_RigVMDispatch_ArrayGetAtIndex(in Array,in Index,out Element)"}.Element
   )
   > &RigUnit_ParentConstraint(
    AdvancedSettings=(InterpolationType=Average,RotationOrderForFilter=XZY)
    Child=(Type=Bone,Name="dog_thigh_r")
    Parents=((Item=(Type=Bone,Name="IK1_dog_thigh_r"),Weight=1))
    Weight=1
   )
   > &RigUnit_ParentConstraint(
    AdvancedSettings=(InterpolationType=Average,RotationOrderForFilter=XZY)
    Child=(Type=Bone,Name="dog_calf_r")
    Parents=((Item=(Type=Bone,Name="IK1_dog_calf_r"),Weight=1))
    Weight=1
   )
   > ComputePoleVector_3 = RigVMFunctionReferenceNode(
    Bone A=(Type=Bone,Name="dog_calf_r")
    Bone B=(Type=Bone,Name="dog_foot_r")
    Bone C=(Type=Bone,Name="dog_ball_r")
    Draw Transform=true
    Initial=false
    OffsetFactor=5
    Origin Scale=(X=0,Y=0,Z=0)
   )
   > &RigUnit_SetTransform(Item=(Type=Bone,Name="PV_R1"), Weight=1, bInitial=false, Value=@ComputePoleVector_3.Transform)
   n18 = &RigUnit_GetRelativeTransformForItem(Child=(Type=Bone,Name="pelvis"), Parent=(Type=Bone,Name="root"), bChildInitial=False, bParentInitial=False)
   Remap = &RigVMFunction_MathDoubleRemap(SourceMaximum=45, SourceMinimum=55, TargetMaximum=1, TargetMinimum=0, Value=@n18.RelativeTransform.Translation.Z)
   Multiply = &RigVMFunction_MathDoubleMul(
    A=@Remap.Result
    B=&RigVMFunction_MathDoubleRemap(SourceMaximum=40, SourceMinimum=35, TargetMaximum=1, TargetMinimum=0, Value=@n18.RelativeTransform.Translation.Z).Result
   )
   > &RigUnit_ModifyTransforms(
    ItemToModify=((Item=(Type=Bone,Name="IK1_dog_ball_r"),Transform=(Rotation=(X=0,Y=0,Z=0,W=1),Translation=(X=0,Y=-10,Z=0),Scale3D=(X=1,Y=1,Z=1))))
    WeightMinimum=0
    Weight=@Multiply.Result
   )
   > &RigUnit_ModifyTransforms(
    ItemToModify="((Item=(Type=Bone,Name=\"pelvis\"),Transform=(Rotation=(X=0,Y=0,Z=0,W=1),Translation=(X=10,Y=0,Z=0),Scale3D=(X=1,Y=1,Z=1))),(Item=(Type=Bone,Name=\"IK1_dog_ball_r\"),Transform=(Rotation=(X=0,Y=0,Z=0,W=1),Translation=(X=0,Y=-12.3418,Z=0),Scale3D=(X=1,Y=1,Z=1))) ...(truncated)"
    WeightMinimum=0
    Weight=&RigVMFunction_MathDoubleRemap(
     SourceMaximum=5
     SourceMinimum=30
     TargetMaximum=1
     TargetMinimum=0
     Value=&RigUnit_GetRelativeTransformForItem(Child=(Type=Bone,Name="pelvis"), Parent=(Type=Bone,Name="root"), bChildInitial=False, bParentInitial=False).RelativeTransform.Translation.Z
    ).Result
   )
   > IKTwoBone_3 = RigVMFunctionReferenceNode(
    Bone A=(Type=Bone,Name="dog_calf_r")
    Bone B=(Type=Bone,Name="dog_foot_r")
    Bone C=(Type=Bone,Name="dog_ball_r")
    Enable Stretch=false
    End Ctrl=(Type=Bone,Name="IK1_dog_ball_r")
    Line Thickness=0.25
    Pole Vector Node=(Type=Bone,Name="PV_R1")
    Primary Axis=(X=1,Y=0,Z=0)
    Secondary Axis=(X=0,Y=1,Z=0)
    Stretch Maximum Ratio=1.25
    Stretch Start Ratio=0.75
    debug=true
    scale=(X=0,Y=0,Z=0)
   )
   &RigUnit_ParentConstraint(
    AdvancedSettings=(InterpolationType=Average,RotationOrderForFilter=XZY)
    Child=(Type=Bone,Name="PV_L")
    Parents=((Item=(Type=Bone,Name="thigh_l"),Weight=1))
    Weight=1
    ExecutePin=@RigVMFunction_Sequence_2.A
   )
   n19 = RigVMVariableNode() {variable=IK1_L_spawned}
   > IKTwoBone = RigVMFunctionReferenceNode(
    Enable Stretch=true
    End Ctrl=(Type=Bone,Name="ctrl_ball_l")
    Line Thickness=0.25
    Pole Vector Node=(Type=Bone,Name="PV_L")
    Primary Axis=(X=-1,Y=0,Z=0)
    Secondary Axis=(X=0,Y=1,Z=0)
    Stretch Maximum Ratio=1.25
    Stretch Start Ratio=0.9
    debug=true
    scale=(X=0,Y=0,Z=0)
    Bone A=RigVMDispatchNode(Index=0, Array=@n19.Value) {template="DISPATCH_RigVMDispatch_ArrayGetAtIndex(in Array,in Index,out Element)"}.Element
    Bone B=RigVMDispatchNode(Index=1, Array=@n19.Value) {template="DISPATCH_RigVMDispatch_ArrayGetAtIndex(in Array,in Index,out Element)"}.Element
    Bone C=RigVMDispatchNode(Index=2, Array=@n19.Value) {template="DISPATCH_RigVMDispatch_ArrayGetAtIndex(in Array,in Index,out Element)"}.Element
   )
   > &RigUnit_ParentConstraint(
    AdvancedSettings=(InterpolationType=Average,RotationOrderForFilter=XZY)
    Child=(Type=Bone,Name="dog_thigh_l")
    Parents=((Item=(Type=Bone,Name="IK1_dog_thigh_l"),Weight=1))
    Weight=1
   )
   > &RigUnit_ParentConstraint(
    AdvancedSettings=(InterpolationType=Average,RotationOrderForFilter=XZY)
    Child=(Type=Bone,Name="dog_calf_l")
    Parents=((Item=(Type=Bone,Name="IK1_dog_calf_l"),Weight=1))
    Weight=1
   )
   > ComputePoleVector_2 = RigVMFunctionReferenceNode(
    Bone A=(Type=Bone,Name="dog_calf_l")
    Bone B=(Type=Bone,Name="dog_foot_l")
    Bone C=(Type=Bone,Name="dog_ball_l")
    Draw Transform=true
    Initial=false
    OffsetFactor=5
    Origin Scale=(X=0,Y=0,Z=0)
   )
   > &RigUnit_SetTransform(Item=(Type=Bone,Name="PV_L1"), Weight=1, bInitial=false, Value=@ComputePoleVector_2.Transform)
   > &RigUnit_ModifyTransforms(
    ItemToModify=((Item=(Type=Bone,Name="IK1_dog_ball_l"),Transform=(Rotation=(X=0,Y=0,Z=0,W=1),Translation=(X=0,Y=-4,Z=0),Scale3D=(X=1,Y=1,Z=1))))
    WeightMinimum=0
    Weight=@Multiply.Result
   )
   > IKTwoBone_1 = RigVMFunctionReferenceNode(
    Bone A=(Type=Bone,Name="dog_calf_l")
    Bone B=(Type=Bone,Name="dog_foot_l")
    Bone C=(Type=Bone,Name="dog_ball_l")
    Enable Stretch=false
    End Ctrl=(Type=Bone,Name="IK1_dog_ball_l")
    Line Thickness=0.25
    Pole Vector Node=(Type=Bone,Name="PV_L1")
    Primary Axis=(X=-1,Y=0,Z=0)
    Secondary Axis=(X=0,Y=-1,Z=0)
    Stretch Maximum Ratio=1.25
    Stretch Start Ratio=0.75
    debug=true
    scale=(X=0,Y=0,Z=0)
   )
   n20 = &RigVMFunction_Sequence(ExecuteContext=@n16.True)
   /Script/ControlRig.RigUnit_AimItem(
    DebugSettings=(bEnabled=False,Scale=10,WorldOffset=(Rotation=(X=0,Y=0,Z=0,W=1),Translation=(X=0,Y=0,Z=0),Scale3D=(X=1,Y=1,Z=1)))
    Item=(Type=Bone,Name="dog_thigh_l")
    Primary=(Weight=1,Axis=(X=-1,Y=0,Z=0),Target=(X=1,Y=0,Z=0),Kind=Location,Space=(Type=Bone,Name="calf_l"))
    Secondary=(Weight=1,Axis=(X=0,Y=0,Z=1),Target=(X=1,Y=0,Z=0),Kind=Location,Space=(Type=Bone,Name="None"))
    Weight=1
    ExecutePin=@n20.A
   ) {is=struct RigVMUnitNode}
   > DeformRig_GetDeltaRefTransform_FNC_2 = RigVMFunctionReferenceNode(Item=(Type=Bone,Name="calf_l"))
   > &RigUnit_ModifyTransforms(
    ItemToModify="((Item=(Type=Bone,Name=\"dog_thigh_l\"),Transform=(Rotation=(X=-0.0074,Y=-0.0856,Z=-0.0438,W=0.9953),Translation=(X=0,Y=0,Z=0),Scale3D=(X=1.2,Y=1,Z=1))),(Item=(Type=Bone,Name=\"dog_calf_l\"),Transform=(Rotation=(X=0,Y=0,Z=0,W=1),Translation=(X=-8,Y=2,Z=0),Scal ...(truncated)"
    WeightMinimum=0
    Weight=&RigVMFunction_MathFloatRemap(SourceMaximum=100, SourceMinimum=10, bClamp=true, Value=@DeformRig_GetDeltaRefTransform_FNC_2.OutTransform.Rotation.Yaw).Result
   )
   > &RigUnit_AimItem(
    DebugSettings=(bEnabled=False,Scale=10,WorldOffset=(Rotation=(X=0,Y=0,Z=0,W=1),Translation=(X=0,Y=0,Z=0),Scale3D=(X=1,Y=1,Z=1)))
    Item=(Type=Bone,Name="dog_calf_l")
    Primary=(Weight=1,Axis=(X=-1,Y=0,Z=0),Target=(X=1,Y=0,Z=0),Kind=Location,Space=(Type=Bone,Name="foot_l"))
    Secondary=(Weight=1,Axis=(X=0,Y=0,Z=1),Target=(X=1,Y=0,Z=0),Kind=Location,Space=(Type=Bone,Name="None"))
    Weight=1
   )
   &RigUnit_AimItem(
    DebugSettings=(bEnabled=False,Scale=10,WorldOffset=(Rotation=(X=0,Y=0,Z=0,W=1),Translation=(X=0,Y=0,Z=0),Scale3D=(X=1,Y=1,Z=1)))
    Item=(Type=Bone,Name="dog_thigh_r")
    Primary=(Weight=1,Axis=(X=1,Y=0,Z=0),Target=(X=1,Y=0,Z=0),Kind=Location,Space=(Type=Bone,Name="calf_r"))
    Secondary=(Weight=1,Axis=(X=0,Y=0,Z=-1),Target=(X=1,Y=0,Z=0),Kind=Location,Space=(Type=Bone,Name="None"))
    Weight=1
    ExecutePin=@n20.B
   )
   > DeformRig_GetDeltaRefTransform_FNC_3 = RigVMFunctionReferenceNode(Item=(Type=Bone,Name="calf_r"))
   > &RigUnit_ModifyTransforms(
    ItemToModify="((Item=(Type=Bone,Name=\"dog_thigh_r\"),Transform=(Rotation=(X=0.023,Y=-0.0584,Z=-0.0421,W=0.9971),Translation=(X=0,Y=0,Z=0),Scale3D=(X=1.2,Y=1,Z=1))),(Item=(Type=Bone,Name=\"dog_calf_r\"),Transform=(Rotation=(X=0,Y=0,Z=0,W=1),Translation=(X=8,Y=-2,Z=0),Scale3 ...(truncated)"
    WeightMinimum=0
    Weight=&RigVMFunction_MathFloatRemap(SourceMaximum=100, SourceMinimum=10, bClamp=true, Value=@DeformRig_GetDeltaRefTransform_FNC_3.OutTransform.Rotation.Yaw).Result
   )
   > &RigUnit_AimItem(
    DebugSettings=(bEnabled=False,Scale=10,WorldOffset=(Rotation=(X=0,Y=0,Z=0,W=1),Translation=(X=0,Y=0,Z=0),Scale3D=(X=1,Y=1,Z=1)))
    Item=(Type=Bone,Name="dog_calf_r")
    Primary=(Weight=1,Axis=(X=1,Y=0,Z=0),Target=(X=1,Y=0,Z=0),Kind=Location,Space=(Type=Bone,Name="foot_r"))
    Secondary=(Weight=1,Axis=(X=0,Y=0,Z=-1),Target=(X=1,Y=0,Z=0),Kind=Location,Space=(Type=Bone,Name="None"))
    Weight=1
   )
   n21 = &RigVMFunction_AlphaInterp(
    Bias=0.1
    InterpSpeedDecreasing=10
    InterpSpeedIncreasing=20
    Scale=1
    bClampResult=true
    bMapRange=false
    Value=RigVMDispatchNode(False=0, True=1, Condition=@n13.Value) {template="DISPATCH_RigVMDispatch_If(in Condition,in True,in False,out Result)"}.Result
   )
   n22 = /Script/ControlRig.RigUnit_GetInitialBoneTransform(Bone=thigh_r, Space=GlobalSpace) {is=struct RigVMUnitNode}
   n23 = &RigUnit_GetInitialBoneTransform(Bone=dog_thigh_r, Space=GlobalSpace)
   /Script/ControlRig.RigUnit_SetBoneRotation(
    Bone=dog_thigh_r
    Space=LocalSpace
    bPropagateToChildren=True
    Weight=@n21.Result
    Rotation=&RigVMFunction_MathTransformMakeRelative(
     Global=&RigUnit_GetBoneTransform(Bone=thigh_r, Space=LocalSpace).Transform
     Parent=&RigVMFunction_MathTransformMakeRelative(Global=@n22.Transform, Parent=@n23.Transform).Local
    ).Local.Rotation
    ExecutePin=@RigVMFunction_Sequence_1.B
   ) {is=struct RigVMUnitNode}
   n24 = &RigUnit_GetInitialBoneTransform(Bone=calf_r, Space=GlobalSpace)
   n25 = &RigUnit_GetInitialBoneTransform(Bone=dog_calf_r, Space=GlobalSpace)
   > &RigUnit_SetBoneRotation(
    Bone=dog_calf_r
    Space=LocalSpace
    bPropagateToChildren=True
    Weight=@n21.Result
    Rotation=&RigVMFunction_MathTransformMakeRelative(
     Global=&RigUnit_GetBoneTransform(Bone=calf_r, Space=LocalSpace).Transform
     Parent=&RigVMFunction_MathTransformMakeRelative(Global=@n24.Transform, Parent=@n25.Transform).Local
    ).Local.Rotation
   )
   n26 = &RigUnit_GetInitialBoneTransform(Bone=foot_r, Space=GlobalSpace)
   n27 = &RigUnit_GetInitialBoneTransform(Bone=dog_ball_r, Space=GlobalSpace)
   > &RigUnit_SetBoneRotation(
    Bone=dog_ball_r
    Space=LocalSpace
    Weight=1
    bPropagateToChildren=True
    Rotation=&RigVMFunction_MathTransformMakeRelative(
     Global=&RigUnit_GetBoneTransform(Bone=foot_r, Space=LocalSpace).Transform
     Parent=&RigVMFunction_MathTransformMakeRelative(Global=@n26.Transform, Parent=@n27.Transform).Local
    ).Local.Rotation
   )
   n28 = &RigVMFunction_MathTransformMul(
    B=&RigUnit_GetBoneTransform(Bone=ball_r).Transform
    A=&RigVMFunction_MathTransformMakeRelative(
     Parent=&RigUnit_GetInitialBoneTransform(Bone=ball_r, Space=GlobalSpace).Transform
     Global=&RigUnit_GetInitialBoneTransform(Bone=dog_toe_r, Space=GlobalSpace).Transform
    ).Local
   ) {template="Multiply::Execute(in A,in B,out Result)"}
   > /Script/ControlRig.RigUnit_FABRIK(
    EffectorBone=dog_toe_r
    MaxIterations=10
    Precision=1
    StartBone=dog_thigh_r
    bPropagateToChildren=True
    bSetEffectorTransform=True
    Weight=@n21.Result
    EffectorTransform=@n28.Result
   ) {is=struct RigVMUnitNode}
   > /Script/ControlRig.RigUnit_SetBoneTransform(Bone=dog_toe_r, Space=GlobalSpace, bPropagateToChildren=True, Weight=@n21.Result, Transform=@n28.Result) {is=struct RigVMUnitNode}
   n29 = &RigUnit_GetInitialBoneTransform(Bone=dog_thigh_l, Space=GlobalSpace)
   n30 = &RigUnit_GetInitialBoneTransform(Bone=thigh_l, Space=GlobalSpace)
   &RigUnit_SetBoneRotation(
    Bone=dog_thigh_l
    Space=LocalSpace
    bPropagateToChildren=True
    Weight=@n21.Result
    Rotation=&RigVMFunction_MathTransformMakeRelative(
     Global=&RigUnit_GetBoneTransform(Bone=thigh_l, Space=LocalSpace).Transform
     Parent=&RigVMFunction_MathTransformMakeRelative(Parent=@n29.Transform, Global=@n30.Transform).Local
    ).Local.Rotation
    ExecutePin=@RigVMFunction_Sequence_1.A
   )
   n31 = &RigUnit_GetInitialBoneTransform(Bone=dog_calf_l, Space=GlobalSpace)
   n32 = &RigUnit_GetInitialBoneTransform(Bone=calf_l, Space=GlobalSpace)
   > &RigUnit_SetBoneRotation(
    Bone=dog_calf_l
    Space=LocalSpace
    bPropagateToChildren=True
    Weight=@n21.Result
    Rotation=&RigVMFunction_MathTransformMakeRelative(
     Global=&RigUnit_GetBoneTransform(Bone=calf_l, Space=LocalSpace).Transform
     Parent=&RigVMFunction_MathTransformMakeRelative(Parent=@n31.Transform, Global=@n32.Transform).Local
    ).Local.Rotation
   )
   n33 = &RigUnit_GetInitialBoneTransform(Bone=foot_l, Space=GlobalSpace)
   n34 = &RigUnit_GetInitialBoneTransform(Bone=dog_ball_l, Space=GlobalSpace)
   > &RigUnit_SetBoneRotation(
    Bone=dog_ball_l
    Space=LocalSpace
    Weight=1
    bPropagateToChildren=True
    Rotation=&RigVMFunction_MathTransformMakeRelative(
     Global=&RigUnit_GetBoneTransform(Bone=foot_l, Space=LocalSpace).Transform
     Parent=&RigVMFunction_MathTransformMakeRelative(Global=@n33.Transform, Parent=@n34.Transform).Local
    ).Local.Rotation
   )
   n35 = &RigVMFunction_MathTransformMul(
    B=&RigUnit_GetBoneTransform(Bone=ball_l).Transform
    A=&RigVMFunction_MathTransformMakeRelative(
     Parent=&RigUnit_GetInitialBoneTransform(Bone=ball_l, Space=GlobalSpace).Transform
     Global=&RigUnit_GetInitialBoneTransform(Bone=dog_toe_l, Space=GlobalSpace).Transform
    ).Local
   ) {template="Multiply::Execute(in A,in B,out Result)"}
   > &RigUnit_FABRIK(
    EffectorBone=dog_toe_l
    MaxIterations=10
    Precision=1
    StartBone=dog_thigh_l
    bPropagateToChildren=True
    bSetEffectorTransform=True
    Weight=@n21.Result
    EffectorTransform=@n35.Result
   )
   > &RigUnit_SetBoneTransform(Bone=dog_toe_l, Space=GlobalSpace, bPropagateToChildren=True, Weight=@n21.Result, Transform=@n35.Result)
   RigVMFunction_Sequence_6 = RigVMAggregateNode(ExecuteContext=@RigVMFunction_Sequence.C)
   DeformRig_GetDeltaRefTransform_FNC = RigVMFunctionReferenceNode(Item=(Type=Bone,Name="dog_calf_l"), ExecuteContext=@RigVMFunction_Sequence_6.A)
   > &RigUnit_ModifyTransforms(
    ItemToModify=((Item=(Type=Bone,Name="dog_deform_knee_out_l"),Transform=(Rotation=(X=0,Y=0,Z=0,W=-1),Translation=(X=32,Y=8,Z=0),Scale3D=(X=1,Y=1.2,Z=1.2))))
    WeightMinimum=0
    Weight=&RigVMFunction_MathFloatRemap(SourceMaximum=120, SourceMinimum=30, bClamp=true, Value=@DeformRig_GetDeltaRefTransform_FNC.OutTransform.Rotation.Yaw).Result
   )
   DeformRig_GetDeltaRefTransform_FNC_1 = RigVMFunctionReferenceNode(Item=(Type=Bone,Name="dog_calf_r"), ExecuteContext=@RigVMFunction_Sequence_6.B)
   > &RigUnit_ModifyTransforms(
    ItemToModify=((Item=(Type=Bone,Name="dog_deform_knee_out_r"),Transform=(Rotation=(X=0,Y=0,Z=0,W=1),Translation=(X=-32,Y=5.9464,Z=0),Scale3D=(X=1,Y=1.2,Z=1.2))))
    WeightMinimum=0
    Weight=&RigVMFunction_MathFloatRemap(SourceMaximum=120, SourceMinimum=30, bClamp=true, Value=@DeformRig_GetDeltaRefTransform_FNC_1.OutTransform.Rotation.Yaw).Result
   )
   n36 = &RigUnit_SphericalPoseReader(
    ActiveRegionScaleFactors=(PositiveWidth=3.5,NegativeWidth=0.5,PositiveHeight=3,NegativeHeight=3)
    ActiveRegionSize=0.1
    Debug=(bDrawDebug=false,DebugScale=25,DebugSegments=20,DebugThickness=0.25)
    DriverAxis=(X=-1,Y=0,Z=0)
    DriverItem=(Type=Bone,Name="thigh_l")
    FalloffRegionScaleFactors=(PositiveWidth=1,NegativeWidth=1,PositiveHeight=1,NegativeHeight=1)
    FalloffSize=0.45
    OptionalParentItem=(Type=Bone)
    RotationOffset=(X=90,Y=0,Z=0)
    ExecutePin=@RigVMFunction_Sequence_6.C
   )
   > &RigUnit_ModifyTransforms(
    ItemToModify="((Item=(Type=Bone,Name=\"dog_deform_groin_l\"),Transform=(Rotation=(X=0,Y=0,Z=0,W=-1),Translation=(X=0,Y=8,Z=0),Scale3D=(X=1,Y=1,Z=1))),(Item=(Type=Bone,Name=\"dog_deform_groin_base_l\"),Transform=(Rotation=(X=0,Y=0,Z=0.1736,W=0.9848),Translation=(X=-8,Y=0,Z=0 ...(truncated)"
    WeightMinimum=0
    Weight=@n36.OutputParam
   )
   > &RigUnit_ModifyTransforms(
    ItemToModify=((Item=(Type=Bone,Name="dog_deform_knee_in_l"),Transform=(Rotation=(X=0,Y=0,Z=-0.2756,W=0.9613),Translation=(X=-5,Y=-11,Z=0),Scale3D=(X=1,Y=1,Z=1))))
    WeightMinimum=0
    Weight=@n36.OutputParam
   )
   n37 = &RigUnit_SphericalPoseReader(
    ActiveRegionScaleFactors=(PositiveWidth=3.5,NegativeWidth=0.5,PositiveHeight=3,NegativeHeight=3)
    ActiveRegionSize=0.1
    Debug=(bDrawDebug=false,DebugScale=25,DebugSegments=20,DebugThickness=0.25)
    DriverAxis=(X=1,Y=0,Z=0)
    DriverItem=(Type=Bone,Name="thigh_r")
    FalloffRegionScaleFactors=(PositiveWidth=1,NegativeWidth=1,PositiveHeight=1,NegativeHeight=1)
    FalloffSize=0.45
    FlipHeightScaling=false
    FlipWidthScaling=true
    OptionalParentItem=(Type=Bone)
    RotationOffset=(X=-90,Y=0,Z=0)
    ExecutePin=@RigVMFunction_Sequence_6.D
   )
   > &RigUnit_ModifyTransforms(
    ItemToModify="((Item=(Type=Bone,Name=\"dog_deform_groin_r\"),Transform=(Rotation=(X=0,Y=0,Z=0,W=-1),Translation=(X=0,Y=-8,Z=0),Scale3D=(X=1,Y=1,Z=1))),(Item=(Type=Bone,Name=\"dog_deform_groin_base_r\"),Transform=(Rotation=(X=0,Y=0,Z=0.1736,W=0.9848),Translation=(X=8,Y=0,Z=0 ...(truncated)"
    WeightMinimum=0
    Weight=@n37.OutputParam
   )
   > &RigUnit_ModifyTransforms(
    ItemToModify=((Item=(Type=Bone,Name="dog_deform_knee_in_r"),Transform=(Rotation=(X=0,Y=0,Z=-0.2756,W=0.9613),Translation=(X=0,Y=0.6492,Z=0),Scale3D=(X=1,Y=1,Z=1))))
    WeightMinimum=0
    Weight=@n37.OutputParam
   )
   n38 = RigVMVariableNode() {variable=isCustomBoots}
   n39 = &RigVMFunction_ControlFlowBranch(ExecuteContext=@RigVMFunction_Sequence_6.E, Condition=@n38.Value)
   &RigUnit_ModifyTransforms(
    ItemToModify="((Item=(Type=Bone,Name=\"patell_l\"),Transform=(Rotation=(X=0,Y=0,Z=0.526,W=-0.8505),Translation=(X=1.2512,Y=-4.5062,Z=0),Scale3D=(X=1,Y=1,Z=1))),(Item=(Type=Bone,Name=\"dog_deform_knee_out_l\"),Transform=(Rotation=(X=0,Y=0,Z=0,W=1),Translation=(X=0,Y=-8,Z=0), ...(truncated)"
    Weight=1
    WeightMinimum=0
    ExecutePin=@n39.True
   )
   > &RigUnit_ModifyTransforms(
    ItemToModify="((Item=(Type=Bone,Name=\"patell_r\"),Transform=(Rotation=(X=0.0332,Y=0.1617,Z=-0.5563,W=0.8144),Translation=(X=-1.8188,Y=5.2144,Z=-0.2821),Scale3D=(X=1,Y=1,Z=1))),(Item=(Type=Bone,Name=\"dog_deform_knee_out_r\"),Transform=(Rotation=(X=0,Y=0,Z=0,W=1),Translatio ...(truncated)"
    Weight=1
    WeightMinimum=0
   )
   n40 = &RigVMFunction_ControlFlowBranch(ExecuteContext=(), Condition=RigVMVariableNode() {variable=parachuting}.Value)
   &RigUnit_ModifyTransforms(
    ItemToModify="((Item=(Type=Bone,Name=\"R_tentacle_01_R1\"),Transform=(Rotation=(X=0,Y=0,Z=-0.1908,W=0.9816),Translation=(X=0,Y=0,Z=0),Scale3D=(X=1,Y=1,Z=1))),(Item=(Type=Bone,Name=\"R_tentacle_01_L1\"),Transform=(Rotation=(X=0,Y=0,Z=-0.1908,W=0.9816),Translation=(X=0,Y=0,Z= ...(truncated)"
    Weight=1
    WeightMinimum=0
    ExecutePin=@n40.True
   )
   FN_DeformRig_FNC = RigVMFunctionReferenceNode(
    DeformRig_ Curve="(deformRig_l_upperarm_up_alpha_72_#1=False,deformRig_r_upperarm_up_alpha_73_#2=False,deformRig_l_upperarm_fr_alpha_74_#3=False,deformRig_r_upperarm_fr_alpha_75_CB51B8 ...(truncated)"
    DeformRig_Bool="(upperarm_1_#4=True,elbow_3_#5=True,wrist_7_#6=True,groin_8_#7=True,glute_17_#8=True,knee_9_40C408324CA261 ...(truncated)"
    DeformRig_Debug=(upperarm_debug_18_#4=False,wrist_debug_20_#6=False,lat_debug_26_#9=False,groinGlute_debug_27_#7=False)
    DeformRig_Value="(upperarm_up_18_#10=(Rotation=(X=0,Y=0,Z=0,W=1),Translation=(X=0,Y=0,Z=8),Scale3D=(X=0.7,Y=1,Z=1)),upperarm_fwd_fr_19_#11=(Rotation=(X=0,Y=0,Z=0,W=1),Translation=(X=0,Y=10,Z=0),Scale3D=(X=0.5,Y=1,Z= ...(truncated)"
    ExecuteContext=()
   )
   RigVMDispatchNode(False=1, True=0.4, Condition=@n38.Value) {template="DISPATCH_RigVMDispatch_If(in Condition,in True,in False,out Result)"}
   &RigVMFunction_MathFloatRemap(SourceMaximum=50, SourceMinimum=60, Value=0, bClamp=true)
   RigVMUnitNode(A=(), B=()) {template="Multiply::Execute(in A,in B,out Result)"}
   RigVMFunction_MathBoolNot = &RigVMFunction_MathBoolNot(Value=RigVMVariableNode() {variable=isCustomBoots}.Value)
   &RigVMFunction_MathBoolNot(Value=RigVMVariableNode() {variable=isCustomBoots}.Value)
   &RigVMFunction_MathQuaternionToEuler(
    RotationOrder=ZYX
    Value=&RigUnit_GetRelativeTransformForItem(Child=(Type=Bone,Name="head"), Parent=(Type=Bone,Name="spine_03"), bChildInitial=False, bParentInitial=False).RelativeTransform.Rotation
   )
   &RigVMFunction_MathQuaternionToEuler(
    RotationOrder=ZYX
    Value=&RigUnit_GetRelativeTransformForItem(Child=(Type=Bone,Name="dog_calf_l"), Parent=(Type=Bone,Name="dog_thigh_l"), bChildInitial=False, bParentInitial=False).RelativeTransform.Rotation
   )
   &RigVMFunction_MathQuaternionToEuler(
    RotationOrder=ZYX
    Value=&RigUnit_GetRelativeTransformForItem(Child=(Type=Bone,Name="dog_calf_r"), Parent=(Type=Bone,Name="dog_thigh_r"), bChildInitial=False, bParentInitial=False).RelativeTransform.Rotation
   )
   RigVMVariableNode() {variable=EndInterpEmote}
  ]}
]
hierarchy [LegAdjust {type=Curve, loc=<0 0 0>, rot=<0 0 0 1>}]
variables [
 Aiming {type=double, default=0}
 EndInterpEmote {type=float, default=0}
 GoingUP {type=double, default=0}
 IK1_L {
  type="TArray<FRigElementKey>"
  default=((Type=Bone,Name="dog_thigh_l"),(Type=Bone,Name="dog_calf_l"),(Type=Bone,Name="dog_foot_l"),(Type=Bone,Name="dog_ball_l"))
 }
 IK1_L_spawned {type="TArray<FRigElementKey>"}
 IK1_R {
  type="TArray<FRigElementKey>"
  default=((Type=Bone,Name="dog_thigh_r"),(Type=Bone,Name="dog_calf_r"),(Type=Bone,Name="dog_foot_r"),(Type=Bone,Name="dog_ball_r"))
 }
 IK1_R_spawned {type="TArray<FRigElementKey>"}
 Toes {type="TArray<FRigElementKey>", default=((Type=Bone,Name="ball_l"),(Type=Bone,Name="ball_r"))}
 isAim {type=bool, default=False}
 isCustomBoots {type=bool, default=False}
 isPlatEmote {type=bool, default=False}
 parachuting {type=bool, default=False}
 parent_var {type=FRigElementKey, default=(Type=Bone,Name="pelvis")}
 skyDiving {type=bool, default=False}
]
