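# Post-process-style AnimBlueprint for the TallChore large male cosmetic body, layering secondary motion and pose fixes over the attached parent's pose.
# Copy Pose feeds a Control Rig, then spring AnimDynamics on tentacle bones and dog-thigh look-ats.
# Bool-driven bone edits collapse dog feet for certain boots, adjust arms when DBNO and shift the tail during ScrubaDub.
# EventGraph updates smoothed aiming, vertical velocity, boot and emote flags.
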AnimBlueprint /BRCosmetics/Characters/Player/Male/Large/Bodies/M_LRG_TallChore/Meshes/M_LRG_TallChore_AnimBP
target_skeleton=&/M_LRG_TallChore_Skeleton
graphs [
 AnimGraph {nodes [
   n1 = "Constraint - dog_thigh_twist_01_r (R)"(Alpha=1, ConstraintWeights_0=1) {
    alphaBoolBlend {blendOption=Linear}
    boneToModify {boneName=dog_thigh_twist_01_r}
    constraintSetup [{transformType=Rotation, targetBone {boneName=thigh_twist_01_r}}]
    constraintWeights [1]
   }
   "Constraint - R_dog_thigh_twist_01_l (R)"(ComponentPose=@n1, Alpha=1, ConstraintWeights_0=1) {
    alphaBoolBlend {blendOption=Linear}
    boneToModify {boneName=R_dog_thigh_twist_01_l}
    constraintSetup [{transformType=Rotation, targetBone {boneName=thigh_twist_01_l}}]
    constraintWeights [1]
   }
   n2 = Use cached pose 'GhostCach'()
   n3 = Use cached pose 'GhostCach'()
   n4 = Get bIsTargeting()
   n5 = Anim Dynamics - Bone: R_tentacle_01_M(
    ComponentPose=Local To Component(LocalPose=Blend Poses by bool(BlendPose_0=@n3, BlendPose_1=@n2, BlendTime_0=0.15, bActiveValue=@n4) {blendType=Linear})
    Alpha=1
    ExternalForce="0,0,0"
   ) {
    angularDampingOverride=0.99
    angularSpringConstant=300
    bAngularSpring=true
    bOverrideAngularDamping=true
    bUseGravityOverride=true
    linearSpringConstant=100
    boundBone {boneName=R_tentacle_01_M}
    componentAppliedLinearAccClamp=<1000 1000 1000>
    componentLinearAccScale=<0.2 0.2 0.2>
    componentLinearVelScale=<0.2 0.2 0.2>
    physicsBodyDefinitions [{
      boundBone {boneName=R_tentacle_01_M}
      boxExtents=<10 10 10>
      constraintSetup {angularLimitsMax=<0 5 10>, angularLimitsMin=<0 -35 -10>, angularTarget=<1 0 0>}
      localJointOffset=<0.6 0 0>
     }]
   }
   n6 = Anim Dynamics - Bone: R_tentacle_03_M(ComponentPose=@n5, Alpha=1, ExternalForce="0,0,0") {
    angularDampingOverride=0.99
    angularSpringConstant=300
    bAngularSpring=true
    bOverrideAngularDamping=true
    bUseGravityOverride=true
    linearSpringConstant=100
    boundBone {boneName=R_tentacle_03_M}
    componentAppliedLinearAccClamp=<1000 1000 1000>
    componentLinearAccScale=<0.2 0.2 0.2>
    componentLinearVelScale=<0.2 0.2 0.2>
    physicsBodyDefinitions [{
      boundBone {boneName=R_tentacle_03_M}
      boxExtents=<10 10 10>
      constraintSetup {angularLimitsMax=<0 10 10>, angularLimitsMin=<0 -25 -10>, angularTarget=<1 0 0>}
      localJointOffset=<0.6 0 0>
     }]
   }
   n7 = Anim Dynamics - Bone: R_tentacle_05_M(ComponentPose=@n6, Alpha=1, ExternalForce="0,0,0") {
    angularDampingOverride=0.99
    angularSpringConstant=300
    bAngularSpring=true
    bOverrideAngularDamping=true
    bUseGravityOverride=true
    linearSpringConstant=100
    boundBone {boneName=R_tentacle_05_M}
    componentAppliedLinearAccClamp=<1000 1000 1000>
    componentLinearAccScale=<0.2 0.2 0.2>
    componentLinearVelScale=<0.2 0.2 0.2>
    physicsBodyDefinitions [{
      boundBone {boneName=R_tentacle_05_M}
      boxExtents=<10 10 10>
      constraintSetup {angularLimitsMax=<0 5 10>, angularLimitsMin=<0 -20 -10>, angularTarget=<1 0 0>}
      localJointOffset=<0.6 0 0>
     }]
   }
   n8 = Anim Dynamics - Bone: R_tentacle_01_L1(ComponentPose=@n7, Alpha=1, ExternalForce="0,0,0") {
    angularDampingOverride=0.99
    angularSpringConstant=300
    bAngularSpring=true
    bOverrideAngularDamping=true
    bUseGravityOverride=true
    linearSpringConstant=100
    boundBone {boneName=R_tentacle_01_L1}
    componentAppliedLinearAccClamp=<1000 1000 1000>
    componentLinearAccScale=<0.2 0.2 0.2>
    componentLinearVelScale=<0.2 0.2 0.2>
    physicsBodyDefinitions [{
      boundBone {boneName=R_tentacle_01_L1}
      boxExtents=<10 10 10>
      constraintSetup {angularLimitsMax=<0 15 4>, angularLimitsMin=<0 -35 -10>, angularTarget=<1 0 0>}
      localJointOffset=<0.6 0 0>
     }]
   }
   n9 = Anim Dynamics - Bone: R_tentacle_03_L1(ComponentPose=@n8, Alpha=1, ExternalForce="0,0,0") {
    angularDampingOverride=0.99
    angularSpringConstant=300
    bAngularSpring=true
    bOverrideAngularDamping=true
    bUseGravityOverride=true
    linearSpringConstant=100
    boundBone {boneName=R_tentacle_03_L1}
    componentAppliedLinearAccClamp=<1000 1000 1000>
    componentLinearAccScale=<0.2 0.2 0.2>
    componentLinearVelScale=<0.2 0.2 0.2>
    physicsBodyDefinitions [{
      boundBone {boneName=R_tentacle_03_L1}
      boxExtents=<10 10 10>
      constraintSetup {angularLimitsMax=<0 10 4>, angularLimitsMin=<0 -25 -10>, angularTarget=<1 0 0>}
      localJointOffset=<0.6 0 0>
     }]
   }
   n10 = Anim Dynamics - Bone: R_tentacle_05_L1(ComponentPose=@n9, Alpha=1, ExternalForce="0,0,0") {
    angularDampingOverride=0.99
    angularSpringConstant=300
    bAngularSpring=true
    bOverrideAngularDamping=true
    bUseGravityOverride=true
    linearSpringConstant=100
    boundBone {boneName=R_tentacle_05_L1}
    componentAppliedLinearAccClamp=<1000 1000 1000>
    componentLinearAccScale=<0.2 0.2 0.2>
    componentLinearVelScale=<0.2 0.2 0.2>
    physicsBodyDefinitions [{
      boundBone {boneName=R_tentacle_05_L1}
      boxExtents=<10 10 10>
      constraintSetup {angularLimitsMax=<0 5 4>, angularLimitsMin=<0 -20 -10>, angularTarget=<1 0 0>}
      localJointOffset=<0.6 0 0>
     }]
   }
   n11 = Anim Dynamics - Bone: R_tentacle_01_L2(ComponentPose=@n10, Alpha=1, ExternalForce="0,0,0") {
    angularDampingOverride=0.99
    angularSpringConstant=300
    bAngularSpring=true
    bOverrideAngularDamping=true
    bUseGravityOverride=true
    linearSpringConstant=100
    boundBone {boneName=R_tentacle_01_L2}
    componentAppliedLinearAccClamp=<1000 1000 1000>
    componentLinearAccScale=<0.2 0.2 0.2>
    componentLinearVelScale=<0.2 0.2 0.2>
    physicsBodyDefinitions [{
      boundBone {boneName=R_tentacle_01_L2}
      boxExtents=<10 10 10>
      constraintSetup {angularLimitsMax=<0 15 10>, angularLimitsMin=<0 -35 -10>, angularTarget=<1 0 0>}
      localJointOffset=<0.6 0 0>
     }]
   }
   n12 = Anim Dynamics - Bone: R_tentacle_03_L2(ComponentPose=@n11, Alpha=1, ExternalForce="0,0,0") {
    angularDampingOverride=0.99
    angularSpringConstant=300
    bAngularSpring=true
    bOverrideAngularDamping=true
    bUseGravityOverride=true
    linearSpringConstant=100
    boundBone {boneName=R_tentacle_03_L2}
    componentAppliedLinearAccClamp=<1000 1000 1000>
    componentLinearAccScale=<0.2 0.2 0.2>
    componentLinearVelScale=<0.2 0.2 0.2>
    physicsBodyDefinitions [{
      boundBone {boneName=R_tentacle_03_L2}
      boxExtents=<10 10 10>
      constraintSetup {angularLimitsMax=<0 10 10>, angularLimitsMin=<0 -25 -10>, angularTarget=<1 0 0>}
      localJointOffset=<0.6 0 0>
     }]
   }
   n13 = Anim Dynamics - Bone: R_tentacle_05_L2(ComponentPose=@n12, Alpha=1, ExternalForce="0,0,0") {
    angularDampingOverride=0.99
    angularSpringConstant=300
    bAngularSpring=true
    bOverrideAngularDamping=true
    bUseGravityOverride=true
    linearSpringConstant=100
    boundBone {boneName=R_tentacle_05_L2}
    componentAppliedLinearAccClamp=<1000 1000 1000>
    componentLinearAccScale=<0.2 0.2 0.2>
    componentLinearVelScale=<0.2 0.2 0.2>
    physicsBodyDefinitions [{
      boundBone {boneName=R_tentacle_05_L2}
      boxExtents=<10 10 10>
      constraintSetup {angularLimitsMax=<0 5 10>, angularLimitsMin=<0 -20 -10>, angularTarget=<1 0 0>}
      localJointOffset=<0.6 0 0>
     }]
   }
   n14 = Anim Dynamics - Bone: R_tentacle_01_R2(ComponentPose=@n13, Alpha=1, ExternalForce="0,0,0") {
    angularDampingOverride=0.99
    angularSpringConstant=300
    bAngularSpring=true
    bOverrideAngularDamping=true
    bUseGravityOverride=true
    linearSpringConstant=100
    boundBone {boneName=R_tentacle_01_R2}
    componentAppliedLinearAccClamp=<1000 1000 1000>
    componentLinearAccScale=<0.2 0.2 0.2>
    componentLinearVelScale=<0.2 0.2 0.2>
    physicsBodyDefinitions [{
      boundBone {boneName=R_tentacle_01_R2}
      boxExtents=<10 10 10>
      constraintSetup {angularLimitsMax=<0 15 10>, angularLimitsMin=<0 -35 -10>, angularTarget=<1 0 0>}
      localJointOffset=<-0.6 0 0>
     }]
   }
   n15 = Anim Dynamics - Bone: R_tentacle_03_R2(ComponentPose=@n14, Alpha=1, ExternalForce="0,0,0") {
    angularDampingOverride=0.99
    angularSpringConstant=300
    bAngularSpring=true
    bOverrideAngularDamping=true
    bUseGravityOverride=true
    linearSpringConstant=100
    boundBone {boneName=R_tentacle_03_R2}
    componentAppliedLinearAccClamp=<1000 1000 1000>
    componentLinearAccScale=<0.2 0.2 0.2>
    componentLinearVelScale=<0.2 0.2 0.2>
    physicsBodyDefinitions [{
      boundBone {boneName=R_tentacle_03_R2}
      boxExtents=<10 10 10>
      constraintSetup {angularLimitsMax=<0 10 10>, angularLimitsMin=<0 -25 -10>, angularTarget=<1 0 0>}
      localJointOffset=<-0.6 0 0>
     }]
   }
   n16 = Anim Dynamics - Bone: R_tentacle_05_R2(ComponentPose=@n15, Alpha=1, ExternalForce="0,0,0") {
    angularDampingOverride=0.99
    angularSpringConstant=300
    bAngularSpring=true
    bOverrideAngularDamping=true
    bUseGravityOverride=true
    linearSpringConstant=100
    boundBone {boneName=R_tentacle_05_R2}
    componentAppliedLinearAccClamp=<1000 1000 1000>
    componentLinearAccScale=<0.2 0.2 0.2>
    componentLinearVelScale=<0.2 0.2 0.2>
    physicsBodyDefinitions [{
      boundBone {boneName=R_tentacle_05_R2}
      boxExtents=<10 10 10>
      constraintSetup {angularLimitsMax=<0 5 10>, angularLimitsMin=<0 -20 -10>, angularTarget=<1 0 0>}
      localJointOffset=<-0.6 0 0>
     }]
   }
   n17 = Anim Dynamics - Bone: R_tentacle_01_R1(ComponentPose=@n16, Alpha=1, ExternalForce="0,0,0") {
    angularDampingOverride=0.99
    angularSpringConstant=300
    bAngularSpring=true
    bOverrideAngularDamping=true
    bUseGravityOverride=true
    linearSpringConstant=100
    boundBone {boneName=R_tentacle_01_R1}
    componentAppliedLinearAccClamp=<1000 1000 1000>
    componentLinearAccScale=<0.2 0.2 0.2>
    componentLinearVelScale=<0.2 0.2 0.2>
    physicsBodyDefinitions [{
      boundBone {boneName=R_tentacle_01_R1}
      boxExtents=<10 10 10>
      constraintSetup {angularLimitsMax=<0 15 4>, angularLimitsMin=<0 -35 -10>, angularTarget=<1 0 0>}
      localJointOffset=<-0.6 0 0>
     }]
   }
   n18 = Anim Dynamics - Bone: R_tentacle_03_R1(ComponentPose=@n17, Alpha=1, ExternalForce="0,0,0") {
    angularDampingOverride=0.99
    angularSpringConstant=300
    bAngularSpring=true
    bOverrideAngularDamping=true
    bUseGravityOverride=true
    linearSpringConstant=100
    boundBone {boneName=R_tentacle_03_R1}
    componentAppliedLinearAccClamp=<1000 1000 1000>
    componentLinearAccScale=<0.2 0.2 0.2>
    componentLinearVelScale=<0.2 0.2 0.2>
    physicsBodyDefinitions [{
      boundBone {boneName=R_tentacle_03_R1}
      boxExtents=<10 10 10>
      constraintSetup {angularLimitsMax=<0 10 4>, angularLimitsMin=<0 -25 -10>, angularTarget=<1 0 0>}
      localJointOffset=<-0.6 0 0>
     }]
   }
   n19 = Anim Dynamics - Bone: R_tentacle_05_R1(ComponentPose=@n18, Alpha=1, ExternalForce="0,0,0") {
    angularDampingOverride=0.99
    angularSpringConstant=300
    bAngularSpring=true
    bOverrideAngularDamping=true
    bUseGravityOverride=true
    linearSpringConstant=100
    boundBone {boneName=R_tentacle_05_R1}
    componentAppliedLinearAccClamp=<1000 1000 1000>
    componentLinearAccScale=<0.2 0.2 0.2>
    componentLinearVelScale=<0.2 0.2 0.2>
    physicsBodyDefinitions [{
      boundBone {boneName=R_tentacle_05_R1}
      boxExtents=<10 10 10>
      constraintSetup {angularLimitsMax=<0 5 4>, angularLimitsMin=<0 -20 -10>, angularTarget=<1 0 0>}
      localJointOffset=<-0.6 0 0>
     }]
   }
   n20 = Look At - Bone: dog_thigh_l(ComponentPose=@n19, Alpha=1) {
    alphaBoolBlend {blendOption=Linear}
    boneToModify {boneName=dog_thigh_l}
    lookAtLocation {x=0}
    lookAtTarget {boneReference {boneName=calf_l}}
    lookAt_Axis {axis=<-1 0>}
   }
   n21 = Look At - Bone: dog_thigh_r(ComponentPose=@n20, Alpha=1) {
    alphaBoolBlend {blendOption=Linear}
    boneToModify {boneName=dog_thigh_r}
    lookAtLocation {x=0}
    lookAtTarget {boneReference {boneName=calf_r}}
    lookAt_Axis {axis=<1 0>}
   }
   n22 = Get CustomBootsEquip() -> (CustomBootsEquip:bool)
   n23 = "Transform (Modify) Bone - Bone: dog_foot_l"!(ComponentPose=@n21, bAlphaBoolEnabled:bool=@n22, Translation="0,0,0", Rotation="0,0,0", Scale="0,0,0") -> (Pose) {
    alphaInputType=Bool
    scaleMode=BMM_Additive
    translationMode=BMM_Additive
    translationSpace=BCS_BoneSpace
    boneToModify {boneName=dog_foot_l}
    scale=<0 0 0>
   }
   n24 = "Transform (Modify) Bone - Bone: dog_foot_r"!(ComponentPose=@n23, bAlphaBoolEnabled:bool=@n22, Translation="0,0,0", Rotation="0,0,0", Scale="0,0,0") -> (Pose) {alphaInputType=Bool, scaleMode=BMM_Additive, translationSpace=BCS_BoneSpace, boneToModify {boneName=dog_foot_r}, scale=<0 0 0>}
   n25 = "Transform (Modify) Bone - Bone: dog_calf_corr_l"!(ComponentPose=@n24, bAlphaBoolEnabled:bool=@n22, Translation="0,0,0", Rotation="0,0,0", Scale="0,0,0") -> (Pose) {
    alphaInputType=Bool
    scaleMode=BMM_Additive
    translationMode=BMM_Additive
    translationSpace=BCS_BoneSpace
    boneToModify {boneName=dog_calf_corr_l}
    scale=<0 0 0>
   }
   n26 = "Transform (Modify) Bone - Bone: dog_calf_corr_r"!(ComponentPose=@n25, bAlphaBoolEnabled:bool=@n22, Translation="0,0,0", Rotation="0,0,0", Scale="0,0,0") -> (Pose) {alphaInputType=Bool, scaleMode=BMM_Additive, translationSpace=BCS_BoneSpace, boneToModify {boneName=dog_calf_corr_r}, scale=<0 0 0>}
   n27 = Get bIsDBNO()
   n28 = "Transform (Modify) Bone - Bone: lowerarm_l"!(ComponentPose=@n26, bAlphaBoolEnabled:bool=@n27, Translation="0,0,0", Rotation="0.000,0.000,-5.268", Scale="1,1,1") -> (Pose) {alphaInputType=Bool, rotationMode=BMM_Additive, translationSpace=BCS_BoneSpace, boneToModify {boneName=lowerarm_l}, rotation {roll=-5.268}}
   n29 = "Transform (Modify) Bone - Bone: phantom_index_01_r"!(ComponentPose=@n28, bAlphaBoolEnabled:bool=@n27, Translation="0,0,0", Rotation="0,0,0", Scale="1,1,1") -> (Pose) {alphaInputType=Bool, rotationMode=BMM_Additive, translationSpace=BCS_BoneSpace, boneToModify {boneName=phantom_index_01_r}}
   n30 = "Transform (Modify) Bone - Bone: lowerarm_r"!(ComponentPose=@n29, bAlphaBoolEnabled:bool=@n27, Translation="0,0,0", Rotation="0.000,0.000,-7.867", Scale="1,1,1") -> (Pose) {alphaInputType=Bool, rotationMode=BMM_Additive, translationSpace=BCS_BoneSpace, boneToModify {boneName=lowerarm_r}, rotation {roll=-7.867}}
   n31 = "Transform (Modify) Bone - Bone: phantom_ring_02_r"!(ComponentPose=@n30, bAlphaBoolEnabled:bool=@n27, Translation="0,0,0", Rotation="-0.000,0.000,25.412", Scale="1,1,1") -> (Pose) {
    alphaInputType=Bool
    rotationMode=BMM_Additive
    translationSpace=BCS_BoneSpace
    boneToModify {boneName=phantom_ring_02_r}
    rotation {roll=25.412}
   }
   n32 = "Transform (Modify) Bone - Bone: phantom_ring_01_r"!(ComponentPose=@n31, bAlphaBoolEnabled:bool=@n27, Translation="0,0,0", Rotation="0.000,0.000,-15.543", Scale="1,1,1") -> (Pose) {
    alphaInputType=Bool
    rotationMode=BMM_Additive
    translationSpace=BCS_BoneSpace
    boneToModify {boneName=phantom_ring_01_r}
    rotation {roll=-15.543}
   }
   n33 = "Transform (Modify) Bone - Bone: hand_r"!(ComponentPose=@n32, bAlphaBoolEnabled:bool=@n27, Translation="0,0,0", Rotation="-0.000,0.000,5.587", Scale="1,1,1") -> (Pose) {alphaInputType=Bool, rotationMode=BMM_Additive, translationSpace=BCS_BoneSpace, boneToModify {boneName=hand_r}, rotation {roll=5.587}}
   n34 = "Get scale tail  scrubadub"() -> ("scale tail  scrubadub":bool)
   n35 = "Transform (Modify) Bone - Bone: Tail_corr"!(ComponentPose=@n33, bAlphaBoolEnabled:bool=@n34, Translation="0,15,0", Rotation="0,0,0", Scale="1,1,1") -> (Pose) {alphaInputType=Bool, rotationMode=BMM_Additive, translationMode=BMM_Additive, boneToModify {boneName=Tail_corr}, translation {y=15}}
   n36 = "Transform (Modify) Bone - Bone: Pelvis_coorskin"!(ComponentPose=@n35, bAlphaBoolEnabled:bool=@n34, Translation="0,10,0", Rotation="0,0,0", Scale="1,1,1") -> (Pose) {alphaInputType=Bool, rotationMode=BMM_Additive, translationMode=BMM_Additive, boneToModify {boneName=Pelvis_coorskin}, translation {y=10}}
   Output Pose(Result=Component To Local(ComponentPose=@n36)) {layerGroup=None}
   n37 = Control Rig!(
    Source=Copy Pose From Mesh() {bUseAttachedParent=true}
    Alpha
    parachuting:bool=Get bIsParachuteOpened()
    isAim:bool=Get bIsTargeting()
    Aiming:double=Get Aiming() -> (Aiming:double)
    GoingUP:double=Get isGoingUp() -> (isGoingUp:double)
    skyDiving:bool=Get bIsSkyDiving()
    isCustomBoots:bool=Get CustomBootsEquip()
    isPlatEmote:bool=Get PlayEmote() -> (PlayEmote:bool)
   ) -> (Pose) {controlRigAssetReference {blueprintRigClass=&/M_LRG_TallChore_CtrlRig.M_LRG_TallChore_CtrlRig_C}}
   Save cached pose 'GhostCach'(Pose=@n37) {cacheName=GhostCach}
  ]}
 EventGraph {nodes [
   n1 = Event BlueprintUpdateAnimation()
   > Set Delta Time X(Delta Time X=@n1.DeltaTimeX)
   > n2 = Sequence!(execute) -> (then_0, then_1, then_2:Exec, then_3:Exec, then_4:Exec)
   n3 = GetAnimInstance(self=GetOwnerSkeletalMeshComponent())
   > Is Valid(InputObject=@n3)
   > n4 = Branch(Condition=Montage_IsPlaying(self=@n3, Montage=/Game/Animation/Game/MainPlayer/Emotes/CattusRoar/Emote_CattusRoar_CMM_M))
   > Do Once(Reset=@n5.then)
   > Set isRoaringAlpha(isRoaringAlpha:double=1) -> (Output_Get:double)
   > Delay(Duration=4)
   > Set isRoaringAlpha(isRoaringAlpha=0)
   n5 = Set isRoaringAlpha(execute=@n4.else, isRoaringAlpha=0)
   n6 = Branch(execute=@n2.then_2, Condition=Get bIsTargeting())
   n7 = Get Delta Time X()
   > Set Aiming(Aiming:double=FInterpTo(Current=Get Aiming(), Target=1, DeltaTime=@n7, InterpSpeed=10)) -> (Output_Get:double)
   Set Aiming(execute=@n6.else, Aiming=FInterpTo(Current=Get Aiming(), Target=0, DeltaTime=@n7, InterpSpeed=10))
   Set isGoingUp(execute=@n2.then_3, isGoingUp:double=Get PawnVelocityZ()) -> (Output_Get:double)
   n8 = Get As Fort Player Pawn Athena() -> (As Fort Player Pawn Athena:Fort Player Pawn Athena Object Reference)
   Is Valid(exec=@n2.then_4, InputObject=@n8)
   n9 = GetDisplayName(Object=GetSkeletalMeshForPartType(self=@n8, PartType=ExtraPart))
   n10 = Contains(SearchIn=@n9, Substring=BarnTowel_Boots)
   > Set CustomBootsEquip(CustomBootsEquip:bool=OR Boolean(A=Contains(SearchIn=@n9, Substring=BootsTemplate_Gadget), B=@n10)) -> (Output_Get:bool)
   > Set tst(tst:bool=@n10) -> (Output_Get:bool)
   n11 = Cast To FortPlayerPawn(execute=@n2.then_1, Object=GetOwningActor())
   > Set As Fort Player Pawn(As Fort Player Pawn:Fort Player Pawn Object Reference=@n11) -> (Output_Get:Fort Player Pawn Object Reference)
   > n12 = Get!(execute:Exec) -> (then:Exec, else:Exec, As Fort Player Pawn:Fort Player Pawn Object Reference)
   > Set is_wearing_jumpboots(is_wearing_jumpboots:bool=HasMatchingGameplayTag!(self:Gameplay Tag Asset Interface Interface=@n12, TagToCheck:Gameplay Tag Structure=(TagName="Item.Gadget.JumpBoots.AnimTracker")) -> (ReturnValue:bool)) -> (Output_Get:bool)
   n13 = GetOwnerAnimInstance!(self:Custom Character Part Anim Instance Object Reference) -> (ReturnValue:Anim Instance Object Reference)
   > "Set scale tail  scrubadub"("scale tail  scrubadub":bool="Equal (Object)"(
     A=GetCurrentActiveMontage!(self:Anim Instance Object Reference=@n13) -> (ReturnValue:Anim Montage Object Reference)
     B=/BRCosmetics/Animation/Game/MainPlayer/Emotes/Scrub_A_Dub/Emote_ScrubaDub_CLM_M
    )) -> (Output_Get:bool)
   n14 = GetCurrentActiveMontage(self=GetOwnerAnimInstance())
   > Set PlayEmote(PlayEmote:bool="Equal Exactly (String)"(A=GetDisplayName(Object=@n14), B=Emote_LogarithmKick_CMM_M)) -> (Output_Get:bool)
   Event OnCharacterCustomizationLoaded()
   n15 = TryGetPawnOwner()
   > Is Valid(InputObject=@n15)
   > n16 = Cast To FortPlayerPawnAthena(Object=@n15)
   > Set As Fort Player Pawn Athena(As Fort Player Pawn Athena:Fort Player Pawn Athena Object Reference=@n16) -> (Output_Get:Fort Player Pawn Athena Object Reference)
   > PrintString!(execute, InString=Hello, bPrintToScreen, bPrintToLog, TextColor, Duration, Key) -> (then)
   CharacterItemDefinitionHasAnyMetaTag(Tags=Make Array())
   IsValid()
   Event OnEmoteStopped!() -> (OutputDelegate, then, MontageItemDef:Fort Item Definition Object Reference) {is=title K2Node_Event}
  ]}
]
variables [
 Aiming {type=number}
 As Fort Player Pawn {type=FortPlayerPawn}
 As Fort Player Pawn Athena {type=FortPlayerPawnAthena}
 CustomBootsEquip {type=boolean}
 Delta Time X {type=number}
 is_wearing_jumpboots {type=boolean}
 isGoingUp {type=number}
 isRoaringAlpha {type=number}
 PlayEmote {type=boolean}
 "scale tail  scrubadub" {type=boolean}
 tst {type=boolean}
]
