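# Physics asset of capsule colliders for the large male TallChore player body, named for keeping its cape from clipping into the body.
# Capsules wrap the pelvis, upper spine bones, arms, hands, thighs, calves and feet, with mirrored left and right limbs and two capsules on the pelvis.
# It has no constraints, so it supplies collision shapes only rather than a jointed ragdoll setup.
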PhysicsAsset /BRCosmetics/Characters/Player/Male/Large/Bodies/M_LRG_TallChore/Meshes/M_LRG_TallChore_CapeColliders
bodies {
 calf_l [{shapeType=Capsule, length=26.4868, radius=12.5184, center=<-16.4529 -2.2427 0>, rotation=<rot 90 0 0>}]
 calf_r [{shapeType=Capsule, length=26.4868, radius=12.5184, center=<16.4529 2.2427 0>, rotation=<rot 90 0 180>}]
 foot_l [{shapeType=Capsule, length=15.0766, radius=10.0396, center=<-3.9164 3.986 0>, rotation=<rot 0 14.9998 90>}]
 foot_r [{shapeType=Capsule, length=15.0766, radius=10.0396, center=<3.9164 -3.986 0>, rotation=<rot 0 14.9998 -90>}]
 hand_l [{shapeType=Capsule, length=9.5015, radius=6.7933, center=<6.2703 0 0>, rotation=<rot 90 0 0>}]
 hand_r [{shapeType=Capsule, length=9.5015, radius=6.7933, center=<-6.2703 0 0>, rotation=<rot 90 0 180>}]
 lowerarm_l [{shapeType=Capsule, length=12.2721, radius=9.624, center=<8.9852 0 1.9278>, rotation=<rot 90 0 0>}]
 lowerarm_r [{shapeType=Capsule, length=12.2721, radius=9.624, center=<-8.9852 0 -1.9278>, rotation=<rot 90 0 180>}]
 pelvis [
  {shapeType=Capsule, length=9.5142, radius=13.3796, center=<-15.2644 -15.9162 0.0052>, rotation=<rot 0 -0.0374 -179.7924>}
  {shapeType=Capsule, length=15, radius=15, shapeName=pelvis_capsule}
 ]
 spine_02 [{shapeType=Capsule, length=15.0617, radius=21.5535, center=<-5.0595 -1.7096 0>}]
 spine_03 [{shapeType=Capsule, length=10, radius=16.0565}]
 spine_04 [{shapeType=Capsule, length=21.5471, radius=20.3326, center=<0 -1.6972 0>}]
 spine_05 [{shapeType=Capsule, length=18.2569, radius=19.2748, center=<0 -1.223 0>}]
 thigh_l [{shapeType=Capsule, length=17.2247, radius=12.8304, center=<-22.1102 0 1.2756>, rotation=<rot 90 0 180>}]
 thigh_r [{shapeType=Capsule, length=17.2247, radius=12.8304, center=<22.1102 0 -1.2756>, rotation=<rot 90 180 180>}]
 upperarm_l [{shapeType=Capsule, length=11.1688, radius=13.7525, center=<10.0455 0 0>, rotation=<rot 90 0 0>}]
 upperarm_r [{shapeType=Capsule, length=11.1688, radius=13.7525, center=<-10.0455 0 0>, rotation=<rot 90 0 180>}]
}
constraints []
